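# Translucent UI material for a NitroDrome background, drawing soft gradient lines with faded edges and a Color parameter for tint.
# Quarter-turn rotated UI UVs pass through a symmetry function into a reflected gradient, with parameters for line scale, offset and power.
# The saturated lines are floored at a minimum opacity, then multiplied by a mirrored linear gradient for edge falloff.
# Both mirroring steps use the deprecated FortUILibrary 2x1 symmetry function.
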Material /NitroDrome/Gameplay/LootObject/UI/Materials/M_NitroDrome_Background
graph {
 properties {
  bAutomaticallySetUsageInEditor=false
  blendMode=BLEND_Translucent
  materialDomain=MD_UI
  pixelDepthOffsetMode=PDOM_Legacy
  usedShadingModels=Default Lit
 }
 nodes [
  n1 = MaterialFunctionCall!(
   UV=MaterialFunctionCall(
    UVs=MaterialFunctionCall() {materialFunction=&GetUserInterfaceUV}.Normalized UV
    "Rotation Angle (0-1)"=Constant() {r=0.25}
   ) {materialFunction=&CustomRotator}.Rotated Values
   Offset
   Rotation
   Scale=ScalarParameter() {defaultValue=0.7, group=Gradient Lines, parameterName=GradientLinesScale}
   MirrorY
  ) {materialFunction=/FortUILibrary/MaterialFunctions/Basics/MF_UI_Symmetry_2x1_DEPRECATED}
  n2 = ScalarParameter() {defaultValue=-0.25, group=Gradient Lines, parameterName=GradientLinesOffset}
  n3 = Saturate(None=Multiply(
    A=ScalarParameter() {defaultValue=1, group=Gradient Lines, parameterName=GradientLinesPower}
    B=MaterialFunctionCall!(UVs=@n1.Result, Offset=@n2, RotationCenter, RotationAngle) {materialFunction=/FortUILibrary/MaterialFunctions/Gradients/MF_UI_ReflectedGradient}.Result
   ))
  n4 = ScalarParameter() {defaultValue=0.2, parameterName=MinOpacityBeforeFalloff}
  n5 = MaterialFunctionCall() {materialFunction=&MF_UI_Symmetry_2x1_DEPRECATED}
  n6 = ScalarParameter() {defaultValue=-0.5, group=Gradient Lines, parameterName=EdgeFalloffGradientStart}
  n7 = ScalarParameter() {defaultValue=0.5, parameterName=EdgeFalloffGradientEnd}
  Saturate(None=Multiply(
    A=Max(A=@n3, B=@n4) {constB=0.2}
    B=MaterialFunctionCall(UVs=@n5.Result, GradientStart=@n6, GradientEnd=@n7) {materialFunction=&MF_UI_LinearGradient}.Result
   ))
  VectorParameter() {parameterName=Color, defaultValue=<rgba 1 1 1 1>}
 ]
}
