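; UI director Blueprint for a NitroDrome loot object, built on the dynamic UI director framework.
; On BeginPlay it gets the owning local player's Fort global UI context and checks whether that context is in a zone.
; The branch leads to no action and the construction script is empty, so the director adds no UI in this graph.
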(event EventBeginPlay
  (if (GlobalUIContext|IsinZone (PlayerController|LocalPlayerSubsystems|GetFortGlobalUIContext (DynamicUIDirector|GetOwningLocalPlayerController)))
    _))

(fn ConstructionScript)
