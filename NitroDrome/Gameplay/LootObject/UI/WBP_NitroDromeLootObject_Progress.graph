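# Progress widget for the NitroDrome loot object, showing a diamond progress bar, center icon and announcement banner as the objective moves through its timer states.
# UMG animations cover timer start, time changes, unlock and restart, using transform, color and material-parameter tracks.
# Graph events switch the announcement and center icon, then play the matching animation.
# On completion, a timer based on the unlock animation's length triggers the complete delegate event.
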WidgetBlueprint /NitroDrome/Gameplay/LootObject/UI/WBP_NitroDromeLootObject_Progress
animations [
 Anim_FadeIn {
  binding_count=1
  length_seconds=0.35
  object=&self:Anim_FadeIn
  sequence {
   bindings [Overlay_Progress {
     id=#1
     possessed_class=Overlay
     tracks [{
       type=MovieSceneFloatTrack
       display_name=Render Opacity
       sections [{type=MovieSceneFloatSection, channels [None {type=MovieSceneScriptingFloatChannel, keys [[0,0],[21,1]]}]}]
      }]
    }]
   display_rate {numerator=60}
  }
 }
 Anim_TimerStart {
  binding_count=12
  length_seconds=3.65
  object=&self:Anim_TimerStart
  sequence {
   bindings [
    Image_MMObj_Icon {
     id=#2
     possessed_class=Image
     tracks [
      {
       type=MovieScene2DTransformTrack
       sections [{
         type=MovieScene2DTransformSection
         end_frame=80
         start_frame=0
         channels [
          Translation.X {type=MovieSceneScriptingFloatChannel, const [0,2]}
          Translation.Y {type=MovieSceneScriptingFloatChannel, keys [[0,0],[30,-7.3747],[65.807,0],[79.999,0]]}
          Angle {type=MovieSceneScriptingFloatChannel, const [0,1]}
          Scale.X {type=MovieSceneScriptingFloatChannel, keys [[0,1],[65.807,0.8],[79.999,1]]}
          Scale.Y {type=MovieSceneScriptingFloatChannel, keys [[0,1],[65.807,0.8],[79.999,1]]}
          Shear.X {type=MovieSceneScriptingFloatChannel, const [0,1]}
          Shear.Y {type=MovieSceneScriptingFloatChannel, const [0,1]}
         ]
        }]
      }
      {
       type=MovieSceneColorTrack
       display_name=Color and Opacity
       sections [{
         type=MovieSceneColorSection
         end_frame=80
         start_frame=0
         channels [
          Color.R {type=MovieSceneScriptingFloatChannel, const [1,2]}
          Color.G {type=MovieSceneScriptingFloatChannel, const [1,2]}
          Color.B {type=MovieSceneScriptingFloatChannel, const [1,2]}
          Color.A {type=MovieSceneScriptingFloatChannel, const [1,2]}
         ]
        }]
      }
     ]
    }
    Overlay_UnlockedIcon {
     id=#3
     possessed_class=Overlay
     tracks [{
       type=MovieSceneFloatTrack
       display_name=Render Opacity
       sections [{type=MovieSceneFloatSection, end_frame=218, start_frame=0, channels [None {type=MovieSceneScriptingFloatChannel, const [0,2]}]}]
      }]
    }
    Image_Diamond_ProgressBar {
     id=#4
     possessed_class=Image
     tracks [{
       type=MovieSceneWidgetMaterialTrack
       sections [{
         type=MovieSceneParameterSection
         end_frame=80
         start_frame=0
         channels [
          BGColorR {type=MovieSceneScriptingFloatChannel, const [0.1144,1]}
          BGColorG {type=MovieSceneScriptingFloatChannel, const [0.1144,1]}
          BGColorB {type=MovieSceneScriptingFloatChannel, const [0.1144,1]}
          BGColorA {type=MovieSceneScriptingFloatChannel, const [1,1]}
          FillColorR {type=MovieSceneScriptingFloatChannel, const [0,1]}
          FillColorG {type=MovieSceneScriptingFloatChannel, const [220,1]}
          FillColorB {type=MovieSceneScriptingFloatChannel, const [255,1]}
          FillColorA {type=MovieSceneScriptingFloatChannel, const [1,1]}
         ]
        }]
      }]
    }
    Text_Intro_AnimOnly {
     id=#5
     possessed_class=CommonTextBlock
     tracks [
      {
       type=MovieSceneFloatTrack
       display_name=Render Opacity
       sections [{
         type=MovieSceneFloatSection
         end_frame=100
         start_frame=0
         channels [None {type=MovieSceneScriptingFloatChannel, thinned_from=5, keys [[0,0],[71,0],[72,0.5],[95,0]]}]
        }]
      }
      {
       type=MovieScene2DTransformTrack
       sections [{
         type=MovieScene2DTransformSection
         end_frame=100
         start_frame=59
         channels [
          Translation.X {type=MovieSceneScriptingFloatChannel, const [0,1]}
          Translation.Y {type=MovieSceneScriptingFloatChannel, const [0,1]}
          Angle {type=MovieSceneScriptingFloatChannel, const [0,1]}
          Scale.X {type=MovieSceneScriptingFloatChannel, keys [[59,1],[72,1],[95,2]]}
          Scale.Y {type=MovieSceneScriptingFloatChannel, keys [[59,1],[72,1],[95,2]]}
          Shear.X {type=MovieSceneScriptingFloatChannel, const [0,1]}
          Shear.Y {type=MovieSceneScriptingFloatChannel, const [0,1]}
         ]
        }]
      }
     ]
    }
    Text_Intro {
     id=#6
     possessed_class=CommonTextBlock
     tracks [{
       type=MovieScene2DTransformTrack
       sections [{
         type=MovieScene2DTransformSection
         end_frame=72
         start_frame=0
         channels [
          Translation.X {type=MovieSceneScriptingFloatChannel, const [0,1]}
          Translation.Y {type=MovieSceneScriptingFloatChannel, const [0,1]}
          Angle {type=MovieSceneScriptingFloatChannel, const [0,1]}
          Scale.X {type=MovieSceneScriptingFloatChannel, keys [[0,2],[51,3],[62,2],[72,1]]}
          Scale.Y {type=MovieSceneScriptingFloatChannel, keys [[0,2],[51,3],[62,2],[72,1]]}
          Shear.X {type=MovieSceneScriptingFloatChannel, const [0,1]}
          Shear.Y {type=MovieSceneScriptingFloatChannel, const [0,1]}
         ]
        }]
      }]
    }
    Image_Glow {
     id=#7
     possessed_class=Image
     tracks [
      {
       type=MovieSceneFloatTrack
       display_name=Render Opacity
       sections [{type=MovieSceneFloatSection, channels [None {type=MovieSceneScriptingFloatChannel, keys [[0,1],[55,1],[72,0]]}]}]
      }
      {
       type=MovieSceneColorTrack
       display_name=Color and Opacity
       sections [{
         type=MovieSceneColorSection
         end_frame=22
         start_frame=0
         channels [
          Color.R {type=MovieSceneScriptingFloatChannel, const [1,1]}
          Color.G {type=MovieSceneScriptingFloatChannel, const [1,1]}
          Color.B {type=MovieSceneScriptingFloatChannel, const [1,1]}
          Color.A {type=MovieSceneScriptingFloatChannel, const [1,1]}
         ]
        }]
      }
     ]
    }
    Image_Unlocked_IconGlow_01 {
     id=#8
     possessed_class=Image
     tracks [{
       type=MovieSceneFloatTrack
       display_name=Render Opacity
       sections [{type=MovieSceneFloatSection, end_frame=22, start_frame=0, channels [None {type=MovieSceneScriptingFloatChannel, const [0,1]}]}]
      }]
    }
    Image_MMObj_Icon_Thunders {
     id=#9
     possessed_class=Image
     tracks [
      {
       type=MovieScene2DTransformTrack
       sections [{
         type=MovieScene2DTransformSection
         end_frame=64
         start_frame=0
         channels [
          Translation.X {type=MovieSceneScriptingFloatChannel, const [0,2]}
          Translation.Y {type=MovieSceneScriptingFloatChannel, keys [[-63.507,0],[30,-26.8946],[45.307,2.9132],[64.228,0]]}
          Angle {type=MovieSceneScriptingFloatChannel, const [0,1]}
          Scale.X {type=MovieSceneScriptingFloatChannel, const [1,1]}
          Scale.Y {type=MovieSceneScriptingFloatChannel, const [1,1]}
          Shear.X {type=MovieSceneScriptingFloatChannel, const [0,1]}
          Shear.Y {type=MovieSceneScriptingFloatChannel, const [0,1]}
         ]
        }]
      }
      {
       type=MovieSceneFloatTrack
       display_name=Render Opacity
       sections [{type=MovieSceneFloatSection, end_frame=59, start_frame=0, channels [None {type=MovieSceneScriptingFloatChannel, keys [[30,0],[48.462,1]]}]}]
      }
      {
       type=MovieSceneColorTrack
       display_name=Color and Opacity
       sections [{
         type=MovieSceneColorSection
         end_frame=80
         start_frame=0
         channels [
          Color.R {type=MovieSceneScriptingFloatChannel, const [1,2]}
          Color.G {type=MovieSceneScriptingFloatChannel, const [1,2]}
          Color.B {type=MovieSceneScriptingFloatChannel, const [1,2]}
          Color.A {type=MovieSceneScriptingFloatChannel, const [1,2]}
         ]
        }]
      }
     ]
    }
    Image_Background {
     id=#10
     possessed_class=Image
     tracks [{
       type=MovieSceneColorTrack
       display_name=Color and Opacity
       sections [{
         type=MovieSceneColorSection
         end_frame=85
         start_frame=0
         channels [
          Color.R {type=MovieSceneScriptingFloatChannel, keys [[0,1],[0,0.4],[85,0.4]]}
          Color.G {type=MovieSceneScriptingFloatChannel, keys [[0,1],[0,0.5],[85,0.5]]}
          Color.B {type=MovieSceneScriptingFloatChannel, keys [[0,1],[0,0.7],[85,0.7]]}
          Color.A {type=MovieSceneScriptingFloatChannel, const [1,2]}
         ]
        }]
      }]
    }
    Overlay_Annoucement {
     id=#11
     possessed_class=Overlay
     tracks [{
       type=MovieSceneFloatTrack
       display_name=Render Opacity
       sections [{
         type=MovieSceneFloatSection
         end_frame=219
         start_frame=0
         channels [None {type=MovieSceneScriptingFloatChannel, keys [[0,0],[54,0],[72,1],[210,1],[219,0]]}]
        }]
      }]
    }
    Image_MMObj_Icon_Chest {
     id=#12
     possessed_class=Image
     tracks [
      {
       type=MovieScene2DTransformTrack
       sections [{
         type=MovieScene2DTransformSection
         end_frame=80
         start_frame=0
         channels [
          Translation.X {type=MovieSceneScriptingFloatChannel, const [0,1]}
          Translation.Y {type=MovieSceneScriptingFloatChannel, keys [[0,0],[45.307,15.3431],[68.961,0]]}
          Angle {type=MovieSceneScriptingFloatChannel, const [0,1]}
          Scale.X {type=MovieSceneScriptingFloatChannel, keys [[0,1],[45.307,0.5],[68.961,1.2],[79.999,1]]}
          Scale.Y {type=MovieSceneScriptingFloatChannel, keys [[0,1],[45.307,0.5],[68.961,1.2],[79.999,1]]}
          Shear.X {type=MovieSceneScriptingFloatChannel, const [0,1]}
          Shear.Y {type=MovieSceneScriptingFloatChannel, const [0,1]}
         ]
        }]
      }
      {
       type=MovieSceneFloatTrack
       display_name=Render Opacity
       sections [{
         type=MovieSceneFloatSection
         end_frame=218
         start_frame=0
         channels [None {type=MovieSceneScriptingFloatChannel, keys [[0,0],[45.307,0],[62.653,1],[218,0.9745]]}]
        }]
      }
      {
       type=MovieSceneColorTrack
       display_name=Color and Opacity
       sections [{
         type=MovieSceneColorSection
         end_frame=80
         start_frame=0
         channels [
          Color.R {type=MovieSceneScriptingFloatChannel, const [1,2]}
          Color.G {type=MovieSceneScriptingFloatChannel, const [1,2]}
          Color.B {type=MovieSceneScriptingFloatChannel, const [1,2]}
          Color.A {type=MovieSceneScriptingFloatChannel, const [1,2]}
         ]
        }]
      }
     ]
    }
   ]
   display_rate {numerator=60}
  }
 }
 Anim_Unlocked {
  binding_count=13
  length_seconds=3.0167
  object=&self:Anim_Unlocked
  sequence {
   bindings [
    Image_MMObj_Icon_Thunders {
     id=#13
     possessed_class=Image
     tracks [
      {
       type=MovieScene2DTransformTrack
       sections [{
         type=MovieScene2DTransformSection
         end_frame=181
         start_frame=0
         channels [
          Translation.X {type=MovieSceneScriptingFloatChannel, const [0,1]}
          Translation.Y {type=MovieSceneScriptingFloatChannel, const [0,1]}
          Angle {type=MovieSceneScriptingFloatChannel, const [0,1]}
          Scale.X {type=MovieSceneScriptingFloatChannel, keys [[9,1],[18,1.5],[115,1.5],[150,1]]}
          Scale.Y {type=MovieSceneScriptingFloatChannel, keys [[9,1],[18,1.5],[115,1.5],[150,1]]}
          Shear.X {type=MovieSceneScriptingFloatChannel, const [0,1]}
          Shear.Y {type=MovieSceneScriptingFloatChannel, const [0,1]}
         ]
        }]
      }
      {
       type=MovieSceneFloatTrack
       display_name=Render Opacity
       sections [{
         type=MovieSceneFloatSection
         end_frame=181
         start_frame=0
         channels [None {type=MovieSceneScriptingFloatChannel, thinned_from=6, keys [[0,1],[9,1],[18,0],[115,0],[150,1]]}]
        }]
      }
     ]
    }
    Image_Unlocked_Glow {
     id=#14
     possessed_class=Image
     tracks [
      {
       type=MovieSceneFloatTrack
       display_name=Render Opacity
       sections [{
         type=MovieSceneFloatSection
         end_frame=181
         start_frame=0
         channels [None {type=MovieSceneScriptingFloatChannel, keys [[0,0],[8,0],[26,1],[42,0],[72,1],[115,1],[150,0]]}]
        }]
      }
      {
       type=MovieSceneColorTrack
       display_name=Color and Opacity
       sections [{
         type=MovieSceneColorSection
         end_frame=120
         start_frame=0
         channels [
          Color.R {type=MovieSceneScriptingFloatChannel, const [0,2]}
          Color.G {type=MovieSceneScriptingFloatChannel, keys [[26,0.6769],[72,1]]}
          Color.B {type=MovieSceneScriptingFloatChannel, keys [[26,1],[72,0.4436]]}
          Color.A {type=MovieSceneScriptingFloatChannel, keys [[26,1],[72,1],[115,0]]}
         ]
        }]
      }
     ]
    }
    Image_Unlocked_IconGlow_02 {
     id=#15
     possessed_class=Image
     tracks [
      {
       type=MovieSceneColorTrack
       display_name=Color and Opacity
       sections [{
         type=MovieSceneColorSection
         end_frame=120
         start_frame=0
         channels [
          Color.R {type=MovieSceneScriptingFloatChannel, const [0.6192,1]}
          Color.G {type=MovieSceneScriptingFloatChannel, const [1,1]}
          Color.B {type=MovieSceneScriptingFloatChannel, const [0,1]}
          Color.A {type=MovieSceneScriptingFloatChannel, thinned_from=5, keys [[26,0],[63,0],[91,0.3],[100,0]]}
         ]
        }]
      }
      {
       type=MovieScene2DTransformTrack
       sections [{
         type=MovieScene2DTransformSection
         end_frame=120
         start_frame=0
         channels [
          Angle {type=MovieSceneScriptingFloatChannel, const [0,1]}
          Scale.X {type=MovieSceneScriptingFloatChannel, keys [[53.999,1],[62.999,1.5],[72,0.8],[91,1]]}
          Scale.Y {type=MovieSceneScriptingFloatChannel, keys [[53.999,1],[62.999,1.5],[72,0.8],[91,1]]}
         ]
        }]
      }
     ]
    }
    Image_MMObj_Icon {
     id=#16
     possessed_class=Image
     tracks [
      {
       type=MovieSceneFloatTrack
       display_name=Render Opacity
       sections [{
         type=MovieSceneFloatSection
         end_frame=181
         start_frame=0
         channels [None {type=MovieSceneScriptingFloatChannel, thinned_from=6, keys [[0,1],[30,1],[39,0],[115,0],[150,1]]}]
        }]
      }
      {
       type=MovieScene2DTransformTrack
       sections [{
         type=MovieScene2DTransformSection
         end_frame=181
         start_frame=0
         channels [
          Translation.X {type=MovieSceneScriptingFloatChannel, const [0,1]}
          Translation.Y {type=MovieSceneScriptingFloatChannel, const [0,1]}
          Angle {type=MovieSceneScriptingFloatChannel, const [0,1]}
          Scale.X {type=MovieSceneScriptingFloatChannel, keys [[0,1],[30,1.5],[115,1.5],[150,1]]}
          Scale.Y {type=MovieSceneScriptingFloatChannel, keys [[0,1],[30,1.5],[115,1.5],[150,1]]}
          Shear.X {type=MovieSceneScriptingFloatChannel, const [0,1]}
          Shear.Y {type=MovieSceneScriptingFloatChannel, const [0,1]}
         ]
        }]
      }
     ]
    }
    Overlay_Annoucement {
     id=#17
     possessed_class=Overlay
     tracks [{
       type=MovieSceneFloatTrack
       display_name=Render Opacity
       sections [{
         type=MovieSceneFloatSection
         end_frame=181
         start_frame=0
         channels [None {type=MovieSceneScriptingFloatChannel, keys [[0,0],[54,0],[72,1]]}]
        }]
      }]
    }
    Text_Completion {
     id=#18
     possessed_class=CommonTextBlock
     tracks [{
       type=MovieScene2DTransformTrack
       sections [{
         type=MovieScene2DTransformSection
         end_frame=72
         start_frame=0
         channels [
          Translation.X {type=MovieSceneScriptingFloatChannel, const [0,1]}
          Translation.Y {type=MovieSceneScriptingFloatChannel, const [0,1]}
          Angle {type=MovieSceneScriptingFloatChannel, const [0,1]}
          Scale.X {type=MovieSceneScriptingFloatChannel, keys [[0,2],[51,3],[62,2],[72,1]]}
          Scale.Y {type=MovieSceneScriptingFloatChannel, keys [[0,2],[51,3],[62,2],[72,1]]}
          Shear.X {type=MovieSceneScriptingFloatChannel, const [0,1]}
          Shear.Y {type=MovieSceneScriptingFloatChannel, const [0,1]}
         ]
        }]
      }]
    }
    Image_Unlocked_IconGlow_01 {
     id=#19
     possessed_class=Image
     tracks [
      {
       type=MovieSceneColorTrack
       display_name=Color and Opacity
       sections [
        {
         type=MovieSceneColorSection
         end_frame=59
         start_frame=59
         channels [
          Color.R {type=MovieSceneScriptingFloatChannel, const [1,1]}
          Color.G {type=MovieSceneScriptingFloatChannel, const [1,1]}
          Color.B {type=MovieSceneScriptingFloatChannel, const [1,1]}
          Color.A {type=MovieSceneScriptingFloatChannel, const [1,1]}
         ]
        }
        {
         type=MovieSceneColorSection
         end_frame=120
         start_frame=0
         channels [
          Color.R {type=MovieSceneScriptingFloatChannel, keys [[0,1],[26,0]]}
          Color.G {type=MovieSceneScriptingFloatChannel, keys [[0,1],[26,0.8243]]}
          Color.B {type=MovieSceneScriptingFloatChannel, const [1,2]}
          Color.A {type=MovieSceneScriptingFloatChannel, keys [[0,0],[26,0.25],[39,0]]}
         ]
        }
       ]
      }
      {
       type=MovieScene2DTransformTrack
       sections [{
         type=MovieScene2DTransformSection
         end_frame=120
         start_frame=0
         channels [
          Translation.X {type=MovieSceneScriptingFloatChannel, const [0,1]}
          Translation.Y {type=MovieSceneScriptingFloatChannel, const [0,1]}
          Angle {type=MovieSceneScriptingFloatChannel, const [0,1]}
          Scale.X {type=MovieSceneScriptingFloatChannel, keys [[0,1],[26,1.2],[39,0]]}
          Scale.Y {type=MovieSceneScriptingFloatChannel, keys [[0,1],[26,1.2],[39,0]]}
          Shear.X {type=MovieSceneScriptingFloatChannel, const [0,1]}
          Shear.Y {type=MovieSceneScriptingFloatChannel, const [0,1]}
         ]
        }]
      }
     ]
    }
    Image_Glow {
     id=#20
     possessed_class=Image
     tracks [
      {
       type=MovieSceneFloatTrack
       display_name=Render Opacity
       sections [{type=MovieSceneFloatSection, channels [None {type=MovieSceneScriptingFloatChannel, keys [[0,1],[55,1],[72,0]]}]}]
      }
      {
       type=MovieSceneColorTrack
       display_name=Color and Opacity
       sections [{
         type=MovieSceneColorSection
         channels [
          Color.R {type=MovieSceneScriptingFloatChannel, const [1,1]}
          Color.G {type=MovieSceneScriptingFloatChannel, const [1,1]}
          Color.B {type=MovieSceneScriptingFloatChannel, const [1,1]}
          Color.A {type=MovieSceneScriptingFloatChannel, const [1,1]}
         ]
        }]
      }
     ]
    }
    Text_Completion_AnimOnly {
     id=#21
     possessed_class=CommonTextBlock
     tracks [
      {
       type=MovieScene2DTransformTrack
       sections [{
         type=MovieScene2DTransformSection
         end_frame=100
         start_frame=59
         channels [
          Translation.X {type=MovieSceneScriptingFloatChannel, const [0,1]}
          Translation.Y {type=MovieSceneScriptingFloatChannel, const [0,1]}
          Angle {type=MovieSceneScriptingFloatChannel, const [0,1]}
          Scale.X {type=MovieSceneScriptingFloatChannel, keys [[59,1],[72,1],[95,2]]}
          Scale.Y {type=MovieSceneScriptingFloatChannel, keys [[59,1],[72,1],[95,2]]}
          Shear.X {type=MovieSceneScriptingFloatChannel, const [0,1]}
          Shear.Y {type=MovieSceneScriptingFloatChannel, const [0,1]}
         ]
        }]
      }
      {
       type=MovieSceneFloatTrack
       display_name=Render Opacity
       sections [{
         type=MovieSceneFloatSection
         end_frame=100
         start_frame=0
         channels [None {type=MovieSceneScriptingFloatChannel, thinned_from=5, keys [[0,0],[71,0],[72,0.5],[95,0]]}]
        }]
      }
     ]
    }
    Image_Diamond_ProgressBar {
     id=#22
     possessed_class=Image
     tracks [{
       type=MovieSceneWidgetMaterialTrack
       sections [{
         type=MovieSceneParameterSection
         channels [
          FillColorR {type=MovieSceneScriptingFloatChannel, keys [[0,220],[21,220],[72,0]]}
          FillColorG {type=MovieSceneScriptingFloatChannel, keys [[0,220],[21,220],[72,0.7231],[150,0.7231]]}
          FillColorB {type=MovieSceneScriptingFloatChannel, keys [[0,220],[21,220],[72,0.0021],[150,0.0021]]}
          FillColorA {type=MovieSceneScriptingFloatChannel, const [220,2]}
          BGColorR {type=MovieSceneScriptingFloatChannel, keys [[0,0.1144],[67,0.1144],[72,0],[180,0]]}
          BGColorG {type=MovieSceneScriptingFloatChannel, keys [[0,0.1144],[67,0.1144],[72,0.7231],[180,0.7231]]}
          BGColorB {type=MovieSceneScriptingFloatChannel, keys [[0,0.1144],[67,0.1144],[72,0.0021],[180,0.0021]]}
          BGColorA {type=MovieSceneScriptingFloatChannel, const [1,3]}
         ]
        }]
      }]
    }
    Image_Unlocked_Icon {
     id=#23
     possessed_class=Image
     tracks [
      {
       type=MovieScene2DTransformTrack
       sections [{
         type=MovieScene2DTransformSection
         end_frame=181
         start_frame=0
         channels [
          Translation.X {type=MovieSceneScriptingFloatChannel}
          Translation.Y {type=MovieSceneScriptingFloatChannel}
          Angle {type=MovieSceneScriptingFloatChannel}
          Scale.X {type=MovieSceneScriptingFloatChannel, keys [[51,1.8],[62,1.8],[72,0.6],[91,1],[115,1],[151,0.5]]}
          Scale.Y {type=MovieSceneScriptingFloatChannel, keys [[51,1.5],[62,1.8],[72,0.6],[91,1],[115,1],[151,0.5]]}
          Shear.X {type=MovieSceneScriptingFloatChannel}
          Shear.Y {type=MovieSceneScriptingFloatChannel}
         ]
        }]
      }
      {
       type=MovieSceneFloatTrack
       display_name=Render Opacity
       sections [{
         type=MovieSceneFloatSection
         end_frame=181
         start_frame=0
         channels [None {type=MovieSceneScriptingFloatChannel, thinned_from=5, keys [[0,0],[51,0],[72,1],[181,1]]}]
        }]
      }
     ]
    }
    Overlay_UnlockedIcon {
     id=#24
     possessed_class=Overlay
     tracks [{
       type=MovieSceneFloatTrack
       display_name=Render Opacity
       sections [{
         type=MovieSceneFloatSection
         end_frame=181
         start_frame=0
         channels [None {type=MovieSceneScriptingFloatChannel, thinned_from=5, keys [[0,0],[8,0],[26,1],[150,1]]}]
        }]
      }]
    }
    Image_MMObj_Icon_Chest {
     id=#25
     possessed_class=Image
     tracks [
      {
       type=MovieSceneFloatTrack
       display_name=Render Opacity
       sections [{
         type=MovieSceneFloatSection
         end_frame=181
         start_frame=0
         channels [None {type=MovieSceneScriptingFloatChannel, thinned_from=5, keys [[0,1],[30,1],[39,0],[115,0]]}]
        }]
      }
      {
       type=MovieScene2DTransformTrack
       sections [{
         type=MovieScene2DTransformSection
         end_frame=181
         start_frame=0
         channels [
          Translation.X {type=MovieSceneScriptingFloatChannel, const [0,1]}
          Translation.Y {type=MovieSceneScriptingFloatChannel, const [0,1]}
          Angle {type=MovieSceneScriptingFloatChannel, const [0,1]}
          Scale.X {type=MovieSceneScriptingFloatChannel, keys [[6,1],[18,1.5],[115,1.5]]}
          Scale.Y {type=MovieSceneScriptingFloatChannel, keys [[6,1],[18,1.5],[115,1.5]]}
          Shear.X {type=MovieSceneScriptingFloatChannel, const [0,1]}
          Shear.Y {type=MovieSceneScriptingFloatChannel, const [0,1]}
         ]
        }]
      }
     ]
    }
   ]
   display_rate {numerator=60}
  }
 }
 Anim_TimeAffectedUpdate {
  binding_count=5
  length_seconds=2.0333
  object=&self:Anim_TimeAffectedUpdate
  sequence {
   bindings [
    Image_MMObj_Icon_Thunders {
     id=#26
     possessed_class=Image
     tracks [
      {
       type=MovieScene2DTransformTrack
       sections [{
         type=MovieScene2DTransformSection
         end_frame=18
         start_frame=0
         channels [
          Translation.X {type=MovieSceneScriptingFloatChannel, const [0,2]}
          Translation.Y {type=MovieSceneScriptingFloatChannel, keys [[0,0],[3,0],[7,3],[11,-8],[15,0],[18,0]]}
          Angle {type=MovieSceneScriptingFloatChannel, const [0,1]}
          Scale.X {type=MovieSceneScriptingFloatChannel, keys [[0,1],[6,1.5],[18,1]]}
          Scale.Y {type=MovieSceneScriptingFloatChannel, keys [[0,1],[6,1.5],[18,1]]}
          Shear.X {type=MovieSceneScriptingFloatChannel, const [0,1]}
          Shear.Y {type=MovieSceneScriptingFloatChannel, const [0,1]}
         ]
        }]
      }
      {
       type=MovieSceneColorTrack
       display_name=Color and Opacity
       sections [{
         type=MovieSceneColorSection
         end_frame=20
         start_frame=0
         channels [
          Color.R {type=MovieSceneScriptingFloatChannel, const [1,2]}
          Color.G {type=MovieSceneScriptingFloatChannel, keys [[0,1],[4,0.6],[14,0.6],[20,1]]}
          Color.B {type=MovieSceneScriptingFloatChannel, keys [[0,1],[4,0],[14,0],[20,1]]}
          Color.A {type=MovieSceneScriptingFloatChannel, const [1,2]}
         ]
        }]
      }
     ]
    }
    Overlay_Annoucement {
     id=#27
     possessed_class=Overlay
     tracks [{
       type=MovieSceneFloatTrack
       display_name=Render Opacity
       sections [{
         type=MovieSceneFloatSection
         end_frame=122
         start_frame=0
         channels [None {type=MovieSceneScriptingFloatChannel, keys [[15,0],[25,1],[110,1],[120,0]]}]
        }]
      }]
    }
    Image_MMObj_Icon {
     id=#28
     possessed_class=Image
     tracks [
      {
       type=MovieSceneColorTrack
       display_name=Color and Opacity
       sections [{
         type=MovieSceneColorSection
         end_frame=20
         start_frame=0
         channels [
          Color.R {type=MovieSceneScriptingFloatChannel, const [1,4]}
          Color.G {type=MovieSceneScriptingFloatChannel, keys [[0,1],[6,0.1912],[14,0.1912],[20,1]]}
          Color.B {type=MovieSceneScriptingFloatChannel, keys [[0,1],[6,0],[14,0],[20,1]]}
          Color.A {type=MovieSceneScriptingFloatChannel, const [1,4]}
         ]
        }]
      }
      {
       type=MovieScene2DTransformTrack
       sections [{
         type=MovieScene2DTransformSection
         end_frame=20
         start_frame=0
         channels [
          Translation.X {type=MovieSceneScriptingFloatChannel, const [0,1]}
          Translation.Y {type=MovieSceneScriptingFloatChannel, const [0,1]}
          Angle {type=MovieSceneScriptingFloatChannel, const [0,1]}
          Scale.X {type=MovieSceneScriptingFloatChannel, keys [[0,1],[10,1.25],[20,1]]}
          Scale.Y {type=MovieSceneScriptingFloatChannel, keys [[0,1],[10,1.25],[20,1]]}
          Shear.X {type=MovieSceneScriptingFloatChannel, const [0,1]}
          Shear.Y {type=MovieSceneScriptingFloatChannel, const [0,1]}
         ]
        }]
      }
     ]
    }
    Image_MMObj_Icon_Chest {
     id=#29
     possessed_class=Image
     tracks [
      {
       type=MovieScene2DTransformTrack
       sections [{
         type=MovieScene2DTransformSection
         end_frame=16
         start_frame=0
         channels [
          Translation.X {type=MovieSceneScriptingFloatChannel, const [0,2]}
          Translation.Y {type=MovieSceneScriptingFloatChannel, keys [[0,0],[8,-5.4743],[16,0]]}
          Angle {type=MovieSceneScriptingFloatChannel, const [0,1]}
          Scale.X {type=MovieSceneScriptingFloatChannel, keys [[0,1],[8,1.5],[16,1]]}
          Scale.Y {type=MovieSceneScriptingFloatChannel, keys [[0,1],[8,1.5],[16,1]]}
          Shear.X {type=MovieSceneScriptingFloatChannel, const [0,1]}
          Shear.Y {type=MovieSceneScriptingFloatChannel, const [0,1]}
         ]
        }]
      }
      {
       type=MovieSceneColorTrack
       display_name=Color and Opacity
       sections [{
         type=MovieSceneColorSection
         end_frame=16
         start_frame=0
         channels [
          Color.R {type=MovieSceneScriptingFloatChannel, keys [[0,1],[4,2],[16,1],[20,1]]}
          Color.G {type=MovieSceneScriptingFloatChannel, keys [[0,1],[4,0.6],[12,0.6],[16,1],[20,1]]}
          Color.B {type=MovieSceneScriptingFloatChannel, keys [[0,1],[4,0],[12,0],[16,1],[20,1]]}
          Color.A {type=MovieSceneScriptingFloatChannel, const [1,4]}
         ]
        }]
      }
     ]
    }
    HorizontalBox {
     id=#30
     possessed_class=HorizontalBox
     tracks [{
       type=MovieScene2DTransformTrack
       sections [{
         type=MovieScene2DTransformSection
         end_frame=30
         start_frame=15
         channels [
          Translation.X {type=MovieSceneScriptingFloatChannel}
          Translation.Y {type=MovieSceneScriptingFloatChannel}
          Angle {type=MovieSceneScriptingFloatChannel}
          Scale.X {type=MovieSceneScriptingFloatChannel, keys [[15,1.5],[30,1]]}
          Scale.Y {type=MovieSceneScriptingFloatChannel, keys [[15,1.5],[30,1]]}
          Shear.X {type=MovieSceneScriptingFloatChannel}
          Shear.Y {type=MovieSceneScriptingFloatChannel}
         ]
        }]
      }]
    }
   ]
   display_rate {numerator=60}
  }
 }
 Anim_RestartTimerBegun {
  binding_count=7
  length_seconds=1
  object=&self:Anim_RestartTimerBegun
  sequence {
   bindings [
    Overlay_Annoucement {
     id=#27
     possessed_class=Overlay
     tracks [{
       type=MovieSceneFloatTrack
       display_name=Render Opacity
       sections [{
         type=MovieSceneFloatSection
         end_frame=60
         start_frame=0
         channels [None {type=MovieSceneScriptingFloatChannel, keys [[0,0],[20,0],[30,1],[60,1]]}]
        }]
      }]
    }
    Image_MMObj_Icon_Chest {
     id=#31
     possessed_class=Image
     tracks [
      {
       type=MovieSceneColorTrack
       display_name=Color and Opacity
       sections [{
         type=MovieSceneColorSection
         end_frame=60
         start_frame=0
         channels [
          Color.R {type=MovieSceneScriptingFloatChannel, keys [[0,1],[20,0.2083],[60,0.2141]]}
          Color.G {type=MovieSceneScriptingFloatChannel, keys [[0,1],[20,0.2083],[60,0.2141]]}
          Color.B {type=MovieSceneScriptingFloatChannel, keys [[0,1],[20,0.2083],[60,0.2141]]}
          Color.A {type=MovieSceneScriptingFloatChannel, const [1,3]}
         ]
        }]
      }
      {
       type=MovieSceneFloatTrack
       display_name=Render Opacity
       sections [{type=MovieSceneFloatSection, end_frame=60, start_frame=0, channels [None {type=MovieSceneScriptingFloatChannel, const [1,2]}]}]
      }
      {
       type=MovieScene2DTransformTrack
       sections [{
         type=MovieScene2DTransformSection
         end_frame=0
         start_frame=0
         channels [
          Translation.X {type=MovieSceneScriptingFloatChannel, const [0,1]}
          Translation.Y {type=MovieSceneScriptingFloatChannel, const [0,1]}
          Angle {type=MovieSceneScriptingFloatChannel, const [0,1]}
          Scale.X {type=MovieSceneScriptingFloatChannel, const [1,1]}
          Scale.Y {type=MovieSceneScriptingFloatChannel, const [1,1]}
          Shear.X {type=MovieSceneScriptingFloatChannel, const [0,1]}
          Shear.Y {type=MovieSceneScriptingFloatChannel, const [0,1]}
         ]
        }]
      }
     ]
    }
    Image_MMObj_Icon {
     id=#32
     possessed_class=Image
     tracks [
      {
       type=MovieSceneColorTrack
       display_name=Color and Opacity
       sections [{
         type=MovieSceneColorSection
         end_frame=60
         start_frame=0
         channels [
          Color.R {type=MovieSceneScriptingFloatChannel, keys [[0,1],[20,0.2083],[60,0.2141]]}
          Color.G {type=MovieSceneScriptingFloatChannel, keys [[0,1],[20,0.2083],[60,0.2141]]}
          Color.B {type=MovieSceneScriptingFloatChannel, keys [[0,1],[20,0.2083],[60,0.2141]]}
          Color.A {type=MovieSceneScriptingFloatChannel, const [1,3]}
         ]
        }]
      }
      {
       type=MovieSceneFloatTrack
       display_name=Render Opacity
       sections [{type=MovieSceneFloatSection, end_frame=60, start_frame=0, channels [None {type=MovieSceneScriptingFloatChannel, const [1,2]}]}]
      }
      {
       type=MovieScene2DTransformTrack
       sections [{
         type=MovieScene2DTransformSection
         end_frame=0
         start_frame=0
         channels [
          Translation.X {type=MovieSceneScriptingFloatChannel, const [0,1]}
          Translation.Y {type=MovieSceneScriptingFloatChannel, const [0,1]}
          Angle {type=MovieSceneScriptingFloatChannel, const [0,1]}
          Scale.X {type=MovieSceneScriptingFloatChannel, const [1,1]}
          Scale.Y {type=MovieSceneScriptingFloatChannel, const [1,1]}
          Shear.X {type=MovieSceneScriptingFloatChannel, const [0,1]}
          Shear.Y {type=MovieSceneScriptingFloatChannel, const [0,1]}
         ]
        }]
      }
     ]
    }
    Image_Glow {
     id=#33
     possessed_class=Image
     tracks [
      {
       type=MovieSceneFloatTrack
       display_name=Render Opacity
       sections [{type=MovieSceneFloatSection, channels [None {type=MovieSceneScriptingFloatChannel, keys [[0,0.5],[20,0.5],[37,0]]}]}]
      }
      {
       type=MovieSceneColorTrack
       display_name=Color and Opacity
       sections [{
         type=MovieSceneColorSection
         channels [
          Color.R {type=MovieSceneScriptingFloatChannel, const [1,1]}
          Color.G {type=MovieSceneScriptingFloatChannel, const [1,1]}
          Color.B {type=MovieSceneScriptingFloatChannel, const [1,1]}
          Color.A {type=MovieSceneScriptingFloatChannel, const [1,1]}
         ]
        }]
      }
     ]
    }
    Text_RemainingTime {
     id=#34
     display_name=Text_RestartingTimer
     possessed_class=CommonTextBlock
     tracks [{
       type=MovieScene2DTransformTrack
       sections [{
         type=MovieScene2DTransformSection
         end_frame=60
         start_frame=0
         channels [
          Translation.X {type=MovieSceneScriptingFloatChannel}
          Translation.Y {type=MovieSceneScriptingFloatChannel}
          Angle {type=MovieSceneScriptingFloatChannel}
          Scale.X {type=MovieSceneScriptingFloatChannel, keys [[0,2],[20,2],[35,1],[60,1]]}
          Scale.Y {type=MovieSceneScriptingFloatChannel, keys [[0,2],[20,2],[35,1],[60,1]]}
          Shear.X {type=MovieSceneScriptingFloatChannel}
          Shear.Y {type=MovieSceneScriptingFloatChannel}
         ]
        }]
      }]
    }
    Image_MMObj_Icon_Thunders {
     id=#35
     possessed_class=Image
     tracks [
      {
       type=MovieSceneColorTrack
       display_name=Color and Opacity
       sections [{
         type=MovieSceneColorSection
         end_frame=60
         start_frame=0
         channels [
          Color.R {type=MovieSceneScriptingFloatChannel, keys [[0,1],[20,0.2083],[60,0.2141]]}
          Color.G {type=MovieSceneScriptingFloatChannel, keys [[0,1],[20,0.2083],[60,0.2141]]}
          Color.B {type=MovieSceneScriptingFloatChannel, keys [[0,1],[20,0.2083],[60,0.2141]]}
          Color.A {type=MovieSceneScriptingFloatChannel, const [1,3]}
         ]
        }]
      }
      {
       type=MovieSceneFloatTrack
       display_name=Render Opacity
       sections [{type=MovieSceneFloatSection, end_frame=0, start_frame=0, channels [None {type=MovieSceneScriptingFloatChannel, const [1,1]}]}]
      }
      {
       type=MovieScene2DTransformTrack
       sections [{
         type=MovieScene2DTransformSection
         end_frame=0
         start_frame=0
         channels [
          Translation.X {type=MovieSceneScriptingFloatChannel, const [0,1]}
          Translation.Y {type=MovieSceneScriptingFloatChannel, const [0,1]}
          Angle {type=MovieSceneScriptingFloatChannel, const [0,1]}
          Scale.X {type=MovieSceneScriptingFloatChannel, const [1,1]}
          Scale.Y {type=MovieSceneScriptingFloatChannel, const [1,1]}
          Shear.X {type=MovieSceneScriptingFloatChannel, const [0,1]}
          Shear.Y {type=MovieSceneScriptingFloatChannel, const [0,1]}
         ]
        }]
      }
     ]
    }
    Image_Diamond_ProgressBar {
     id=#36
     possessed_class=Image
     tracks [{
       type=MovieSceneWidgetMaterialTrack
       sections [{
         type=MovieSceneParameterSection
         end_frame=60
         start_frame=0
         channels [
          BGColorR {type=MovieSceneScriptingFloatChannel, const [0.1144,2]}
          BGColorG {type=MovieSceneScriptingFloatChannel, const [0.1144,2]}
          BGColorB {type=MovieSceneScriptingFloatChannel, const [0.1144,2]}
          BGColorA {type=MovieSceneScriptingFloatChannel, const [1,2]}
          FillColorR {type=MovieSceneScriptingFloatChannel, const [1,1]}
          FillColorG {type=MovieSceneScriptingFloatChannel, const [1,1]}
          FillColorB {type=MovieSceneScriptingFloatChannel, const [1,1]}
          FillColorA {type=MovieSceneScriptingFloatChannel, const [0,1]}
         ]
        }]
      }]
    }
   ]
   display_rate {numerator=60}
  }
 }
 Anim_HideBanner {
  binding_count=1
  length_seconds=0.35
  object=&self:Anim_HideBanner
  sequence {bindings [Overlay_Annoucement {
     id=#37
     possessed_class=Overlay
     tracks [{
       type=MovieSceneFloatTrack
       display_name=Render Opacity
       sections [{type=MovieSceneFloatSection, channels [None {type=MovieSceneScriptingFloatChannel, keys [[0,1],[7,0]]}]}]
      }]
    }]}
 }
]
graphs [
 CompleteAnimFinished {nodes [CompleteAnimFinished() {is=title K2Node_FunctionEntry}]}
 EventGraph {nodes [
   Event PlayCompleteAnimation() {is=title K2Node_CustomEvent}
   > SwitchAnnouncementType!(execute, self, Widget:Widget Object Reference=Get VBox_CompletedAnnouncement() -> (VBox_CompletedAnnouncement:Vertical Box Object Reference)) -> (then)
   > n1 = SetActiveWidget(
    self=Get VisSwitcher_CenterIcon() -> (VisSwitcher_CenterIcon:Common Visibility Switcher Object Reference)
    Widget=Get Overlay_UnlockedIcon() -> (Overlay_UnlockedIcon:Overlay Object Reference)
   )
   n2 = Event Call Complete Delegate() {is=title K2Node_CustomEvent}
   n3 = Get Anim_Unlocked() -> (Anim_Unlocked:Widget Animation Object Reference)
   n4 = Set Timer by Event(execute=@n1.then, Delegate=@n2, Time=float + float(A=GetEndTime(self=@n3), B=0.3), bLooping=false)
   > Set Timer_CompleteAnimFinish(Timer_CompleteAnimFinish:Timer Handle Structure=@n4) -> (Output_Get:Timer Handle Structure)
   > PlayAnimation(InAnimation=Get Anim_Unlocked())
   Event OnObjectiveActivated() {is=title K2Node_CustomEvent}
   > SwitchAnnouncementType(Widget=Get VBox_DromeStarted() -> (VBox_DromeStarted:Vertical Box Object Reference))
   > SetActiveWidget(self=Get VisSwitcher_CenterIcon(), Widget=Get Image_MMObj_Icon_Chest() -> (Image_MMObj_Icon_Chest:Image Object Reference))
   > "Play Animation with Finished event (legacy)"!(
    execute:Exec
    Widget:User Widget Object Reference=Self-Reference()
    InAnimation:Widget Animation Object Reference=Get Anim_TimerStart() -> (Anim_TimerStart:Widget Animation Object Reference)
    StartAtTime:float=0
    NumLoopsToPlay:int=1
    PlayMode:EUMGSequencePlayMode Enum
    PlaybackSpeed:float=1
   ) -> (then:Exec, Result:UMGSequence Player Object Reference, Finished:Exec) {is=title K2Node_PlayAnimation}
   > SetUnlockedState!(execute, self, IsUnlocked:bool=false) -> (then)
   OnAnimUnlockedFinished() {is=title K2Node_CustomEvent}
   > Branch(Condition=Get IsUnlocked() -> (IsUnlocked:bool))
   > SetRenderOpacity(self=Get Overlay_Annoucement() -> (Overlay_Annoucement:Overlay Object Reference), InOpacity=0)
   Event UpdateRestartingTimer() {is=title K2Node_CustomEvent}
   n5 = GetSafeZoneIndicator!() -> (ReturnValue:Fort Safe Zone Indicator Object Reference)
   > UpdateResetTimer!(
    execute
    self
    InTimeLeft:double=float - float(
     A=Get SafeZoneFinishShrinkTime!(self:Fort Safe Zone Indicator Object Reference=@n5) -> (SafeZoneFinishShrinkTime:float)
     B=GetServerWorldTimeSeconds!() -> (ReturnValue:float)
    )
   ) -> (then)
   Event ShowRestartingBanner() {is=title K2Node_CustomEvent}
   > StopAnimation(InAnimation=Get Anim_HideBanner() -> (Anim_HideBanner:Widget Animation Object Reference))
   > SwitchAnnouncementType(Widget=Get VBox_RestartTimer() -> (VBox_RestartTimer:Vertical Box Object Reference))
   > SetActiveWidget(self=Get VisSwitcher_CenterIcon(), Widget=Get Image_MMObj_Icon_Chest())
   > SetUnlockedState(IsUnlocked=false)
   > n6 = "Play Animation with Finished event (legacy)"(
    Widget=Self-Reference()
    InAnimation=Get Anim_RestartTimerBegun() -> (Anim_RestartTimerBegun:Widget Animation Object Reference)
    StartAtTime=0
    NumLoopsToPlay=1
    PlaybackSpeed=1
   )
   UpdateClaimProgress!(execute=@n6.Finished, self, ProgressPercent:double=0) -> (then)
   > SetProgressClaimingColor()
   n7 = Event UpdateProgress!() -> (OutputDelegate, then, ProgressPercent:double=~) {is=title K2Node_CustomEvent}
   > UpdateClaimProgress(ProgressPercent=@n7.ProgressPercent)
   Event OnInitialized()
   > BindToAnimationFinished(Animation=Get Anim_Unlocked(), Delegate=Create Event())
   Event Construct()
   > SetProgressClaimingColor()
   Clear and Invalidate Timer by Handle(execute=@n8.then, Handle=Get Timer_CompleteAnimFinish() -> (Timer_CompleteAnimFinish:Timer Handle Structure))
   n8 = Call Complete Anim Finished!(execute:Exec=@n9.then, self:WBP Nitro Drome Loot Object Progress Object Reference) -> (then:Exec) {is=title K2Node_CallDelegate}
   n9 = Set IsUnlocked(execute=@n2.then, IsUnlocked:bool=true) -> (Output_Get:bool)
  ]}
 OnTimeAffectedEventTriggered {nodes [
   n1 = OnTimeAffectedEventTriggered!() -> (then, InTimeDelta:double=~, InPlayer:Actor Object Reference, InEventTags:Gameplay Tag Container Structure) {is=title K2Node_FunctionEntry}
   n2 = Get Text_TimeReduction() -> (Text_TimeReduction:Common Text Object Reference)
   > "SetText (Text)"(
    self=@n2
    InText=Format Text!(
     Format="LOCTEXT(\"{ReducedTime}s\")"
     ReducedTime:Text="To Text (Float)"(Value=@n1, MinimumIntegralDigits=1, MaximumIntegralDigits=324, MinimumFractionalDigits=0, MaximumFractionalDigits=0)
    ) -> (Result)
   )
   > SetVisibility(self=@n2, InVisibility=HitTestInvisible)
   > n3 = SetVisibility(self=Get HBox_PlayerName() -> (HBox_PlayerName:Horizontal Box Object Reference), InVisibility=Hidden)
   n4 = For Each Loop(
    Exec=[@n10.then, @n11.then, @n3.then]
    Array:Array of Gameplay Tag Structures=BreakGameplayTagContainer!(GameplayTagContainer:"Gameplay Tag Container Structure (by ref)"=Get InEventTags() -> (InEventTags:Gameplay Tag Container Structure)) -> (GameplayTags:Array of Gameplay Tag Structures)
   ) -> (Array Element:Gameplay Tag Structure)
   n5 = Find(
    TargetMap:Map of Gameplay Tag Structures to Texts=Get EventTagName() -> (EventTagName:Map of Gameplay Tag Structures to Texts)
    Key:"Gameplay Tag Structure (by ref)"=@n4
   ) -> (Value:Text)
   > Branch(Condition=@n5.ReturnValue)
   > "SetText (Text)"(self=Get Text_TriggeredEvent() -> (Text_TriggeredEvent:Common Text Object Reference), InText=@n5)
   SwitchAnnouncementType(execute=@n4.Completed, Widget=Get VBox_TimeAffectingEvent() -> (VBox_TimeAffectingEvent:Vertical Box Object Reference))
   > PlayTimeAffectedAnims()
   n6 = Get InPlayer() -> (InPlayer:Actor Object Reference)
   n7 = Is Valid(InputObject=@n6)
   > "SetText (Text)"(self=Get Text_Playername() -> (Text_Playername:Common Text Object Reference), InText="To Text (String)"(InString=GetDisplayName(Object=@n6)))
   > n8 = Branch(Condition=HasTag!(
     TagContainer:"Gameplay Tag Container Structure (by ref)"=Get InEventTags()
     Tag:Gameplay Tag Structure=(TagName="NitroDrome.Event.PawnEliminated.Player")
     bExactMatch:bool=false
    ) -> (ReturnValue:bool))
   n9 = Get HBox_PlayerName()
   n10 = SetVisibility(execute=[@n8.then, @n7.Is Not Valid], self=@n9, InVisibility=Hidden)
   n11 = SetVisibility(execute=@n8.else, self=@n9, InVisibility=HitTestInvisible)
  ]}
 OnWidgetHidden {nodes [
   OnWidgetHidden() {is=title K2Node_FunctionEntry}
   > Clear and Invalidate Timer by Handle(Handle=Get Timer_RestartCountdown() -> (Timer_RestartCountdown:Timer Handle Structure))
   > Clear and Invalidate Timer by Handle(Handle=Get Timer_CompleteAnimFinish())
  ]}
 PlayClaimingStartedAnims {nodes [
   PlayClaimingStartedAnims() {is=title K2Node_FunctionEntry}
   > PlayAnimation(InAnimation=Get Anim_TimerStart())
  ]}
 PlayFadeInAnim {nodes [
   PlayFadeInAnim() {is=title K2Node_FunctionEntry}
   > SetRenderOpacity(self=Get Overlay_Annoucement(), InOpacity=0)
   > StopAllAnimations(self:User Widget Object Reference)
   > PlayAnimation(InAnimation=Get Anim_FadeIn() -> (Anim_FadeIn:Widget Animation Object Reference))
  ]}
 PlayHideNitroDromeBanner {nodes [
   PlayHideNitroDromeBanner() {is=title K2Node_FunctionEntry}
   n1 = Get Overlay_Annoucement()
   > Branch(Condition="float > float"(A=GetRenderOpacity!(self:Widget Object Reference=@n1) -> (ReturnValue:float)))
   > PlayAnimation(InAnimation=Get Anim_HideBanner())
  ]}
 PlayRestartTimerBegunAnim {nodes [
   PlayRestartTimerBegunAnim() {is=title K2Node_FunctionEntry}
   > PlayAnimation(InAnimation=Get Anim_RestartTimerBegun())
  ]}
 PlayTimeAffectedAnims {nodes [
   PlayTimeAffectedAnims() {is=title K2Node_FunctionEntry}
   > PlayAnimation(InAnimation=Get Anim_TimeAffectedUpdate() -> (Anim_TimeAffectedUpdate:Widget Animation Object Reference))
  ]}
 SetProgressClaimingColor {nodes [
   SetProgressClaimingColor() {is=title K2Node_FunctionEntry}
   > n1 = GetDynamicMaterial(self=Get Image_Diamond_ProgressBar() -> (Image_Diamond_ProgressBar:Image Object Reference))
   > SetVectorParameterValue(self=@n1, ParameterName=FillColor, Value=Get ProgressColor() -> (ProgressColor:Linear Color Structure))
   > SetVectorParameterValue(self=@n1, ParameterName=BGColor, Value=Get BGColor() -> (BGColor:Linear Color Structure))
  ]}
 SetUnlockedState {nodes [
   n1 = SetUnlockedState!() -> (then, IsUnlocked:bool=~) {is=title K2Node_FunctionEntry}
   > n2 = Delta Gate Wildcard!(execute:Exec, Variable:"Boolean (by ref)"=Get IsUnlocked(), New Value:bool=@n1) -> (Success:Exec, Updated Variable:"Boolean (by ref)", Previous Variable:bool=~) {is=title K2Node_MacroInstance}
   > Branch(Condition=@n2)
   n3 = Get Anim_Unlocked()
   > PlayAnimation(InAnimation=@n3, StartAtTime=GetEndTime(self=@n3))
  ]}
 SwitchAnnouncementType {nodes [
   n1 = SwitchAnnouncementType!() -> (then, Widget:Widget Object Reference) {is=title K2Node_FunctionEntry}
   > SetActiveWidget(self=Get VisSwitcher_AnnoucementType() -> (VisSwitcher_AnnoucementType:Common Visibility Switcher Object Reference), Widget=@n1)
   > n2 = Branch(Condition="Equal (Object)"(A=@n1, B=Get VBox_RestartTimer()))
   > n3 = Set Timer by Event(Delegate=Create Event(), Time=1, bLooping=true)
   > Set Timer_RestartCountdown(Timer_RestartCountdown:Timer Handle Structure=@n3) -> (Output_Get:Timer Handle Structure)
   > Event UpdateRestartingTimer() {is=title K2Node_CallFunction}
   Clear and Invalidate Timer by Handle(execute=@n2.else, Handle=Get Timer_RestartCountdown())
  ]}
 UpdateClaimProgress {nodes [
   n1 = UpdateClaimProgress!() -> (then, ProgressPercent:double=~) {is=title K2Node_FunctionEntry}
   > n2 = GetDynamicMaterial(self=Get Image_Diamond_ProgressBar())
   > SetScalarParameterValue(self=@n2, ParameterName=Progress, Value=@n1)
  ]}
 UpdateResetTimer {nodes [
   n1 = UpdateResetTimer!() -> (then, InTimeLeft:double=~) {is=title K2Node_FunctionEntry}
   n2 = float / float(A=@n1, B=60)
   n3 = "% (Float)"!(A:double=@n1, B:double=60) -> (ReturnValue:double)
   > "SetText (Text)"(
    self=Get Text_RemainingTime() -> (Text_RemainingTime:Common Text Object Reference)
    InText=Format Text!(
     Format="LOCTEXT(\"{minutes}:{seconds}\")"
     minutes:int=Truncate(A=@n2)
     seconds:Text="To Text (Float)"(Value=@n3, RoundingMode=ToZero, MinimumIntegralDigits=2, MaximumIntegralDigits=324, MinimumFractionalDigits=0, MaximumFractionalDigits=0)
    ) -> (Result)
   )
  ]}
]
variables [
 BGColor {type=LinearColor, default=<rgba 0.1144 0.1144 0.1144 1>}
 CachedLocalPawn {type=Pawn, default=null}
 CompleteAnimFinished {}
 CompletedColor {type=LinearColor, default=<rgba 0 0.6795 0.0015 1>}
 IsUnlocked {type=boolean}
 ProgressColor {type=LinearColor, default=<rgba 0 1 1 1>}
 SF_ObjectiveDuration {
  type=ScalableFloat
  default {
   registryType=None
   value=1
   curve {curveTable=/NitroDrome/DataTables/NitroDromeGameData, rowName=NitroDrome.Objective.Interaction.ObjectiveCountdown}
  }
 }
 Timer_CompleteAnimFinish {type=TimerHandle}
 Timer_RestartCountdown {type=TimerHandle}
]
widgets {
 info {inheritedWidgetCount=0, parentClass=&CommonUserWidget, rootWidgetClass=&Overlay, widgetCount=39}
 widgets [
  Overlay_Root {widgetClassPath=&Overlay, widget=Overlay_Root}
  Overlay_DromeTimer {
   widgetClassPath=&Overlay
   bIsVariable=true
   parent=Overlay_Root
   widget=Overlay_DromeTimer
   slot {class=OverlaySlot, properties {horizontalAlignment=HAlign_Center, verticalAlignment=VAlign_Center}}
  }
  FortVisualAttachment_0 {
   widgetClassPath=&FortVisualAttachment
   parent=Overlay_DromeTimer
   widget=FortVisualAttachment_0
   properties {contentAnchor=<-0.05 0.65>, renderTransform {angle=0, scale=<1 1>, shear=<0 0>, translation=<70 0>}}
   slot {class=OverlaySlot, properties {horizontalAlignment=HAlign_Center, verticalAlignment=VAlign_Center}}
  }
  Overlay_Annoucement {
   widgetClassPath=&Overlay
   bIsVariable=true
   parent=FortVisualAttachment_0
   widget=Overlay_Annoucement
   properties {renderOpacity=0}
   slot {class=SizeBoxSlot, properties {horizontalAlignment=HAlign_Center, verticalAlignment=VAlign_Center}}
  }
  Image_Glow {
   widgetClassPath=&Image
   bIsVariable=true
   parent=Overlay_Annoucement
   widget=Image_Glow
   properties {
    renderOpacity=0.2
    brush {
     bIsDynamicallyLoaded=false
     drawAs=Image
     imageType=NoImage
     mirroring=NoMirror
     resourceName=None
     resourceObject=/Game/Athena/HUD/LTM/Wax/Materials/M_UI_TextGlow_Add
     tiling=NoTile
     imageSize=<32 32>
     margin=<ltrb 0 0 0 0>
     outlineSettings {
      bUseBrushTransparency=false
      roundingType=HalfHeightRadius
      width=0
      color {colorUseRule=UseColor_Specified, specifiedColor=<rgba 0 0 0 0>}
      cornerRadii=<0 0 0 0>
     }
     tintColor {colorUseRule=UseColor_Specified, specifiedColor=<rgba 1 1 1 1>}
     uVRegion {bIsValid=false, max=<0 0>, min=<0 0>}
    }
    colorAndOpacity=<rgba 0 0.0103 0.1301 0.5>
    renderTransform {angle=0, scale=<1 1>, shear=<0 -4>, translation=<0 0>}
   }
   slot {class=OverlaySlot, properties {horizontalAlignment=HAlign_Fill, verticalAlignment=VAlign_Fill, padding=<ltrb -75 -32 -75 -32>}}
  }
  Image_Background {
   widgetClassPath=&Image
   parent=Overlay_Annoucement
   widget=Image_Background
   properties {
    brush {
     bIsDynamicallyLoaded=false
     drawAs=Image
     imageType=FullColor
     mirroring=NoMirror
     resourceName=None
     resourceObject=/NitroDrome/Gameplay/LootObject/UI/Materials/MI_NitroDrome_TimeAffecting_Background
     tiling=NoTile
     imageSize=<32 32>
     margin=<ltrb 0 0 0 0>
     outlineSettings {
      bUseBrushTransparency=false
      roundingType=HalfHeightRadius
      width=0
      color {colorUseRule=UseColor_Specified, specifiedColor=<rgba 0 0 0 0>}
      cornerRadii=<0 0 0 0>
     }
     tintColor {colorUseRule=UseColor_Specified, specifiedColor=<rgba 1 1 1 1>}
     uVRegion {bIsValid=false, max=<0 0>, min=<0 0>}
    }
    colorAndOpacity=<rgba 0.4 0.5 0.7 1>
    renderTransform {angle=0, scale=<1 1>, shear=<0 -4>, translation=<0 0>}
   }
   slot {class=OverlaySlot, properties {horizontalAlignment=HAlign_Fill, verticalAlignment=VAlign_Fill, padding=<ltrb -75 -5 -75 -5>}}
  }
  VisSwitcher_AnnoucementType {
   widgetClassPath=&CommonVisibilitySwitcher
   bIsVariable=true
   parent=Overlay_Annoucement
   widget=VisSwitcher_AnnoucementType
   slot {class=OverlaySlot, properties {horizontalAlignment=HAlign_Fill, verticalAlignment=VAlign_Fill}}
  }
  VBox_DromeStarted {
   widgetClassPath=&VerticalBox
   bIsVariable=true
   parent=VisSwitcher_AnnoucementType
   widget=VBox_DromeStarted
   properties {renderTransform {angle=0, scale=<1 1>, shear=<-8 -4>, translation=<0 0>}}
   slot {class=CommonVisibilitySwitcherSlot}
  }
  Overlay_0 {
   widgetClassPath=&Overlay
   parent=VBox_DromeStarted
   widget=Overlay_0
   slot {class=VerticalBoxSlot, properties {horizontalAlignment=HAlign_Left, verticalAlignment=VAlign_Center}}
  }
  Text_Intro_AnimOnly {
   widgetClassPath=&CommonTextBlock
   parent=Overlay_0
   widget=Text_Intro_AnimOnly
   properties {
    justification=Center
    lineHeightPercentage=0.8
    text="Survive the \r\nNitrodrome!"
    textTransformPolicy=ToUpper
    font {
     bForceMonospaced=false
     bMaterialIsStencil=false
     fontMaterial=null
     fontObject=&BurbankSmall
     letterSpacing=0
     monospacedWidth=1
     size=15
     skewAmount=0
     typefaceFontName=Black
     outlineSettings {
      bApplyOutlineToDropShadows=false
      bMiteredCorners=false
      bSeparateFillAlpha=false
      outlineMaterial=null
      outlineSize=0
      outlineColor=<rgba 0 0 0 1>
     }
    }
    shadowColorAndOpacity=<rgba 0 0.0103 0.1301 0.8>
    shadowOffset=<3 4>
   }
   slot {class=OverlaySlot, properties {verticalAlignment=VAlign_Center}}
  }
  Text_Intro {
   widgetClassPath=&CommonTextBlock
   parent=Overlay_0
   widget=Text_Intro
   properties {
    justification=Center
    lineHeightPercentage=0.8
    text="Survive the \r\nNitrodrome!"
    textTransformPolicy=ToUpper
    font {
     bForceMonospaced=false
     bMaterialIsStencil=false
     fontMaterial=null
     fontObject=&BurbankSmall
     letterSpacing=0
     monospacedWidth=1
     size=15
     skewAmount=0
     typefaceFontName=Black
     outlineSettings {
      bApplyOutlineToDropShadows=false
      bMiteredCorners=false
      bSeparateFillAlpha=false
      outlineMaterial=null
      outlineSize=0
      outlineColor=<rgba 0 0 0 1>
     }
    }
    shadowColorAndOpacity=<rgba 0 0.0103 0.1301 0.8>
    shadowOffset=<3 4>
   }
   slot {class=OverlaySlot}
  }
  VBox_TimeAffectingEvent {
   widgetClassPath=&VerticalBox
   bIsVariable=true
   parent=VisSwitcher_AnnoucementType
   widget=VBox_TimeAffectingEvent
   properties {renderTransform {angle=0, scale=<1 1>, shear=<-8 -4>, translation=<0 0>}}
   slot {class=CommonVisibilitySwitcherSlot}
  }
  HorizontalBox {widgetClassPath=&HorizontalBox, parent=VBox_TimeAffectingEvent, widget=HorizontalBox, slot {class=VerticalBoxSlot}}
  Text_TimeReduction {
   widgetClassPath=&CommonTextBlock
   bIsVariable=true
   parent=HorizontalBox
   widget=Text_TimeReduction
   properties {
    text=-30s
    textTransformPolicy=ToLower
    font {
     bForceMonospaced=false
     bMaterialIsStencil=false
     fontMaterial=null
     fontObject=&BurbankSmall
     letterSpacing=0
     monospacedWidth=1
     size=28.5
     skewAmount=0
     typefaceFontName=Black
     outlineSettings {
      bApplyOutlineToDropShadows=false
      bMiteredCorners=false
      bSeparateFillAlpha=false
      outlineMaterial=null
      outlineSize=0
      outlineColor=<rgba 0 0 0 1>
     }
    }
    shadowColorAndOpacity=<rgba 0 0.0103 0.1301 0.8>
    shadowOffset=<3 4>
   }
   slot {class=HorizontalBoxSlot, properties {horizontalAlignment=HAlign_Center, verticalAlignment=VAlign_Center, padding=<ltrb 0 0 5 0>}}
  }
  Text_TriggeredEvent {
   widgetClassPath=&CommonTextBlock
   bIsVariable=true
   parent=HorizontalBox
   widget=Text_TriggeredEvent
   properties {
    autoWrapText=true
    lineHeightPercentage=0.8
    text=Player Eliminated
    textTransformPolicy=ToUpper
    wrapTextAt=175
    font {
     bForceMonospaced=false
     bMaterialIsStencil=false
     fontMaterial=null
     fontObject=&BurbankSmall
     letterSpacing=0
     monospacedWidth=1
     size=15
     skewAmount=0
     typefaceFontName=Black
     outlineSettings {
      bApplyOutlineToDropShadows=false
      bMiteredCorners=false
      bSeparateFillAlpha=false
      outlineMaterial=null
      outlineSize=0
      outlineColor=<rgba 0 0 0 1>
     }
    }
    shadowColorAndOpacity=<rgba 0 0.0103 0.1301 0.8>
    shadowOffset=<3 4>
   }
   slot {class=HorizontalBoxSlot, properties {horizontalAlignment=HAlign_Left, verticalAlignment=VAlign_Center}}
  }
  HBox_PlayerName {
   widgetClassPath=&HorizontalBox
   bIsVariable=true
   parent=VBox_TimeAffectingEvent
   widget=HBox_PlayerName
   slot {class=VerticalBoxSlot, properties {horizontalAlignment=HAlign_Center, verticalAlignment=VAlign_Top, padding=<ltrb 0 0 0 -12>}}
  }
  Text_By {
   widgetClassPath=&CommonTextBlock
   parent=HBox_PlayerName
   widget=Text_By
   properties {
    autoWrapText=true
    justification=Center
    lineHeightPercentage=0.8
    text=BY
    font {
     bForceMonospaced=false
     bMaterialIsStencil=false
     fontMaterial=null
     fontObject=&BurbankSmall
     letterSpacing=0
     monospacedWidth=1
     size=10.5
     skewAmount=0
     typefaceFontName=Black
     outlineSettings {
      bApplyOutlineToDropShadows=false
      bMiteredCorners=false
      bSeparateFillAlpha=false
      outlineMaterial=null
      outlineSize=0
      outlineColor=<rgba 0 0 0 1>
     }
    }
    shadowColorAndOpacity=<rgba 0 0.0103 0.1301 0.8>
    shadowOffset=<3 4>
   }
   slot {class=HorizontalBoxSlot, properties {padding=<ltrb 0 0 5 0>}}
  }
  Text_Playername {
   widgetClassPath=&CommonTextBlock
   bIsVariable=true
   parent=HBox_PlayerName
   widget=Text_Playername
   properties {
    autoWrapText=true
    justification=Center
    lineHeightPercentage=0.8
    text="[PLAYERNAME]"
    font {
     bForceMonospaced=false
     bMaterialIsStencil=false
     fontMaterial=null
     fontObject=&BurbankSmall
     letterSpacing=0
     monospacedWidth=1
     size=10.5
     skewAmount=0
     typefaceFontName=Black
     outlineSettings {
      bApplyOutlineToDropShadows=false
      bMiteredCorners=false
      bSeparateFillAlpha=false
      outlineMaterial=null
      outlineSize=0
      outlineColor=<rgba 0 0 0 1>
     }
    }
    shadowColorAndOpacity=<rgba 0 0.0103 0.1301 0.8>
    shadowOffset=<3 4>
   }
   slot {class=HorizontalBoxSlot}
  }
  VBox_CompletedAnnouncement {
   widgetClassPath=&VerticalBox
   bIsVariable=true
   parent=VisSwitcher_AnnoucementType
   widget=VBox_CompletedAnnouncement
   properties {renderTransform {angle=0, scale=<1 1>, shear=<-8 -4>, translation=<0 0>}}
   slot {class=CommonVisibilitySwitcherSlot}
  }
  Overlay_1 {
   widgetClassPath=&Overlay
   parent=VBox_CompletedAnnouncement
   widget=Overlay_1
   slot {class=VerticalBoxSlot, properties {horizontalAlignment=HAlign_Left, verticalAlignment=VAlign_Center}}
  }
  Text_Completion {
   widgetClassPath=&CommonTextBlock
   parent=Overlay_1
   widget=Text_Completion
   properties {
    justification=Center
    lineHeightPercentage=0.8
    text="Congratulations!\r\nClaim Your Loot!"
    textTransformPolicy=ToUpper
    font {
     bForceMonospaced=false
     bMaterialIsStencil=false
     fontMaterial=null
     fontObject=&BurbankSmall
     letterSpacing=0
     monospacedWidth=1
     size=15
     skewAmount=0
     typefaceFontName=Black
     outlineSettings {
      bApplyOutlineToDropShadows=false
      bMiteredCorners=false
      bSeparateFillAlpha=false
      outlineMaterial=null
      outlineSize=0
      outlineColor=<rgba 0 0 0 1>
     }
    }
    shadowColorAndOpacity=<rgba 0 0.0103 0.1301 0.8>
    shadowOffset=<3 4>
   }
   slot {class=OverlaySlot}
  }
  Text_Completion_AnimOnly {
   widgetClassPath=&CommonTextBlock
   parent=Overlay_1
   widget=Text_Completion_AnimOnly
   properties {
    justification=Center
    lineHeightPercentage=0.8
    text="Congratulations!\r\nClaim your loot!"
    textTransformPolicy=ToUpper
    font {
     bForceMonospaced=false
     bMaterialIsStencil=false
     fontMaterial=null
     fontObject=&BurbankSmall
     letterSpacing=0
     monospacedWidth=1
     size=15
     skewAmount=0
     typefaceFontName=Black
     outlineSettings {
      bApplyOutlineToDropShadows=false
      bMiteredCorners=false
      bSeparateFillAlpha=false
      outlineMaterial=null
      outlineSize=0
      outlineColor=<rgba 0 0 0 1>
     }
    }
    shadowColorAndOpacity=<rgba 0 0.0103 0.1301 0.8>
    shadowOffset=<3 4>
   }
   slot {class=OverlaySlot, properties {verticalAlignment=VAlign_Center}}
  }
  VBox_RestartTimer {
   widgetClassPath=&VerticalBox
   bIsVariable=true
   parent=VisSwitcher_AnnoucementType
   widget=VBox_RestartTimer
   properties {renderTransform {angle=0, scale=<1 1>, shear=<-8 -4>, translation=<0 0>}}
   slot {class=CommonVisibilitySwitcherSlot, properties {padding=<ltrb 0 0 0 -5>}}
  }
  Text_Restarting {
   widgetClassPath=&CommonTextBlock
   parent=VBox_RestartTimer
   widget=Text_Restarting
   properties {
    justification=Center
    lineHeightPercentage=0.8
    text=Restarting in...
    textTransformPolicy=ToUpper
    font {
     bForceMonospaced=false
     bMaterialIsStencil=false
     fontMaterial=null
     fontObject=&BurbankSmall
     letterSpacing=0
     monospacedWidth=1
     size=15
     skewAmount=0
     typefaceFontName=Black
     outlineSettings {
      bApplyOutlineToDropShadows=false
      bMiteredCorners=false
      bSeparateFillAlpha=false
      outlineMaterial=null
      outlineSize=0
      outlineColor=<rgba 0 0 0 1>
     }
    }
    shadowColorAndOpacity=<rgba 0 0.0103 0.1301 0.8>
    shadowOffset=<3 4>
   }
   slot {class=VerticalBoxSlot}
  }
  Text_RemainingTime {
   widgetClassPath=&CommonTextBlock
   bIsVariable=true
   parent=VBox_RestartTimer
   widget=Text_RemainingTime
   properties {
    justification=Center
    text=00:00
    textTransformPolicy=ToUpper
    font {
     bForceMonospaced=false
     bMaterialIsStencil=false
     fontMaterial=null
     fontObject=&BurbankSmall
     letterSpacing=0
     monospacedWidth=1
     size=15
     skewAmount=0
     typefaceFontName=Black
     outlineSettings {
      bApplyOutlineToDropShadows=false
      bMiteredCorners=false
      bSeparateFillAlpha=false
      outlineMaterial=null
      outlineSize=0
      outlineColor=<rgba 0 0 0 1>
     }
    }
    shadowColorAndOpacity=<rgba 0 0.0103 0.1301 0.8>
    shadowOffset=<3 4>
   }
   slot {class=VerticalBoxSlot}
  }
  Overlay_Progress {widgetClassPath=&Overlay, bIsVariable=true, parent=Overlay_DromeTimer, widget=Overlay_Progress, slot {class=OverlaySlot}}
  Image_Unlocked_Glow {
   widgetClassPath=&Image
   parent=Overlay_Progress
   widget=Image_Unlocked_Glow
   properties {
    renderOpacity=0
    brush {
     bIsDynamicallyLoaded=false
     drawAs=Image
     imageType=NoImage
     mirroring=NoMirror
     resourceName=None
     resourceObject=/FortUILibrary/Styles/Border/Materials/M_Border_SoftGlow
     tiling=NoTile
     imageSize=<64 64>
     margin=<ltrb 0 0 0 0>
     outlineSettings {
      bUseBrushTransparency=false
      roundingType=HalfHeightRadius
      width=0
      color {colorUseRule=UseColor_Specified, specifiedColor=<rgba 0 0 0 0>}
      cornerRadii=<0 0 0 0>
     }
     tintColor {colorUseRule=UseColor_Specified, specifiedColor=<rgba 1 1 1 1>}
     uVRegion {bIsValid=false, max=<0 0>, min=<0 0>}
    }
    colorAndOpacity=<rgba 0.2472 1 0 1>
   }
   slot {class=OverlaySlot, properties {horizontalAlignment=HAlign_Fill, verticalAlignment=VAlign_Fill, padding=<ltrb -5 -5 -5 -5>}}
  }
  Overlay_TimerProgress {
   widgetClassPath=&Overlay
   bIsVariable=true
   parent=Overlay_Progress
   widget=Overlay_TimerProgress
   slot {class=OverlaySlot, properties {horizontalAlignment=HAlign_Fill, verticalAlignment=VAlign_Fill, padding=<ltrb 15 15 15 15>}}
  }
  Image_Diamond_BGOpaque {
   widgetClassPath=&Image
   parent=Overlay_TimerProgress
   widget=Image_Diamond_BGOpaque
   properties {brush {
     bIsDynamicallyLoaded=false
     drawAs=Image
     imageType=NoImage
     mirroring=NoMirror
     resourceName=None
     resourceObject=/Game/Athena/HUD/Compass/M_UI_DiamondMarker
     tiling=NoTile
     imageSize=<32 32>
     margin=<ltrb 0 0 0 0>
     outlineSettings {
      bUseBrushTransparency=false
      roundingType=HalfHeightRadius
      width=0
      color {colorUseRule=UseColor_Specified, specifiedColor=<rgba 0 0 0 0>}
      cornerRadii=<0 0 0 0>
     }
     tintColor {colorUseRule=UseColor_Specified, specifiedColor=<rgba 0 0 0 0.5>}
     uVRegion {bIsValid=false, max=<0 0>, min=<0 0>}
    }}
   slot {class=OverlaySlot, properties {horizontalAlignment=HAlign_Fill, verticalAlignment=VAlign_Fill}}
  }
  Image_Diamond_ProgressBar {
   widgetClassPath=&Image
   bIsVariable=true
   parent=Overlay_TimerProgress
   widget=Image_Diamond_ProgressBar
   properties {
    visibility=SelfHitTestInvisible
    brush {
     bIsDynamicallyLoaded=false
     drawAs=Image
     imageType=NoImage
     mirroring=NoMirror
     resourceName=None
     resourceObject=/Game/Athena/HUD/Materials/M_UI_DiamondProgress_RadialFill_Inst
     tiling=NoTile
     imageSize=<68 68>
     margin=<ltrb 0 0 0 0>
     outlineSettings {
      bUseBrushTransparency=false
      roundingType=HalfHeightRadius
      width=0
      color {colorUseRule=UseColor_Specified, specifiedColor=<rgba 0 0 0 0>}
      cornerRadii=<0 0 0 0>
     }
     tintColor {colorUseRule=UseColor_Specified, specifiedColor=<rgba 1 1 1 1>}
     uVRegion {bIsValid=false, max=<0 0>, min=<0 0>}
    }
    renderTransform {angle=-45, scale=<1 1>, shear=<0 0>, translation=<0 0>}
   }
   slot {class=OverlaySlot, properties {horizontalAlignment=HAlign_Center, verticalAlignment=VAlign_Center}}
  }
  Image_Unlocked_IconGlow_01 {
   widgetClassPath=&Image
   parent=Overlay_Progress
   widget=Image_Unlocked_IconGlow_01
   properties {
    brush {
     bIsDynamicallyLoaded=false
     drawAs=Image
     imageType=NoImage
     mirroring=NoMirror
     resourceName=None
     resourceObject=/MidMatchAssignedGameplay/Items/Environmental/Materials/MI_MMObj_Timer_Glow
     tiling=NoTile
     imageSize=<84 84>
     margin=<ltrb 0 0 0 0>
     outlineSettings {
      bUseBrushTransparency=false
      roundingType=HalfHeightRadius
      width=0
      color {colorUseRule=UseColor_Specified, specifiedColor=<rgba 0 0 0 0>}
      cornerRadii=<0 0 0 0>
     }
     tintColor {colorUseRule=UseColor_Specified, specifiedColor=<rgba 1 1 1 1>}
     uVRegion {bIsValid=false, max=<0 0>, min=<0 0>}
    }
    colorAndOpacity=<rgba 0 0.8243 1 0>
   }
   slot {class=OverlaySlot, properties {horizontalAlignment=HAlign_Center, verticalAlignment=VAlign_Center, padding=<ltrb -5 -5 -5 -5>}}
  }
  Overlay_DromeIcon {
   widgetClassPath=&Overlay
   parent=Overlay_Progress
   widget=Overlay_DromeIcon
   slot {class=OverlaySlot, properties {horizontalAlignment=HAlign_Center, verticalAlignment=VAlign_Center}}
  }
  Image_MMObj_Icon {
   widgetClassPath=&Image
   bIsVariable=true
   parent=Overlay_DromeIcon
   widget=Image_MMObj_Icon
   properties {brush {
     bIsDynamicallyLoaded=false
     drawAs=Image
     imageType=FullColor
     mirroring=NoMirror
     resourceName=None
     resourceObject=/NitroDrome/Gameplay/LootObject/UI/Textures/T_UI_Icon_NitroDrome_Timer_Chest
     tiling=NoTile
     imageSize=<64 64>
     margin=<ltrb 0 0 0 0>
     outlineSettings {
      bUseBrushTransparency=false
      roundingType=HalfHeightRadius
      width=0
      color {colorUseRule=UseColor_Specified, specifiedColor=<rgba 0 0 0 0>}
      cornerRadii=<0 0 0 0>
     }
     tintColor {colorUseRule=UseColor_Specified, specifiedColor=<rgba 1 1 1 1>}
     uVRegion {bIsValid=false, max=<0 0>, min=<0 0>}
    }}
   slot {class=OverlaySlot, properties {horizontalAlignment=HAlign_Center, verticalAlignment=VAlign_Center}}
  }
  VisSwitcher_CenterIcon {
   widgetClassPath=&CommonVisibilitySwitcher
   bIsVariable=true
   parent=Overlay_DromeIcon
   widget=VisSwitcher_CenterIcon
   slot {class=OverlaySlot, properties {horizontalAlignment=HAlign_Center, verticalAlignment=VAlign_Center}}
  }
  Image_MMObj_Icon_Chest {
   widgetClassPath=&Image
   bIsVariable=true
   parent=VisSwitcher_CenterIcon
   widget=Image_MMObj_Icon_Chest
   properties {brush {
     bIsDynamicallyLoaded=false
     drawAs=Image
     imageType=FullColor
     mirroring=NoMirror
     resourceName=None
     resourceObject=/Game/Athena/UI/InGame/Creative/PropertyEditor/Icon_Picker/Textures_V2/64x64/T_UI_IconLibrary_Chest_64
     tiling=NoTile
     imageSize=<32 32>
     margin=<ltrb 0 0 0 0>
     outlineSettings {
      bUseBrushTransparency=false
      roundingType=HalfHeightRadius
      width=0
      color {colorUseRule=UseColor_Specified, specifiedColor=<rgba 0 0 0 0>}
      cornerRadii=<0 0 0 0>
     }
     tintColor {colorUseRule=UseColor_Specified, specifiedColor=<rgba 1 1 1 1>}
     uVRegion {bIsValid=false, max=<0 0>, min=<0 0>}
    }}
   slot {class=CommonVisibilitySwitcherSlot}
  }
  Overlay_UnlockedIcon {
   widgetClassPath=&Overlay
   bIsVariable=true
   parent=VisSwitcher_CenterIcon
   widget=Overlay_UnlockedIcon
   slot {class=CommonVisibilitySwitcherSlot, properties {horizontalAlignment=HAlign_Center, verticalAlignment=VAlign_Center}}
  }
  Image_Unlocked_Icon {
   widgetClassPath=&Image
   parent=Overlay_UnlockedIcon
   widget=Image_Unlocked_Icon
   properties {brush {
     bIsDynamicallyLoaded=false
     drawAs=Image
     imageType=NoImage
     mirroring=NoMirror
     resourceName=None
     resourceObject=/Game/Athena/UI/InGame/Creative/PropertyEditor/Icon_Picker/Textures/128x128/T_UI_IconLibrary_unlocked_128
     tiling=NoTile
     imageSize=<56 56>
     margin=<ltrb 0 0 0 0>
     outlineSettings {
      bUseBrushTransparency=false
      roundingType=HalfHeightRadius
      width=0
      color {colorUseRule=UseColor_Specified, specifiedColor=<rgba 0 0 0 0>}
      cornerRadii=<0 0 0 0>
     }
     tintColor {colorUseRule=UseColor_Specified, specifiedColor=<rgba 1 1 1 1>}
     uVRegion {bIsValid=false, max=<0 0>, min=<0 0>}
    }}
   slot {class=OverlaySlot, properties {horizontalAlignment=HAlign_Center, verticalAlignment=VAlign_Center}}
  }
  Image_Unlocked_IconGlow_02 {
   widgetClassPath=&Image
   parent=Overlay_UnlockedIcon
   widget=Image_Unlocked_IconGlow_02
   properties {
    brush {
     bIsDynamicallyLoaded=false
     drawAs=Image
     imageType=NoImage
     mirroring=NoMirror
     resourceName=None
     resourceObject=&MI_MMObj_Timer_Glow
     tiling=NoTile
     imageSize=<84 84>
     margin=<ltrb 0 0 0 0>
     outlineSettings {
      bUseBrushTransparency=false
      roundingType=HalfHeightRadius
      width=0
      color {colorUseRule=UseColor_Specified, specifiedColor=<rgba 0 0 0 0>}
      cornerRadii=<0 0 0 0>
     }
     tintColor {colorUseRule=UseColor_Specified, specifiedColor=<rgba 1 1 1 1>}
     uVRegion {bIsValid=false, max=<0 0>, min=<0 0>}
    }
    colorAndOpacity=<rgba 0.6192 1 0 0>
   }
   slot {class=OverlaySlot, properties {horizontalAlignment=HAlign_Center, verticalAlignment=VAlign_Center, padding=<ltrb -5 -5 -5 -5>}}
  }
  Image_MMObj_Icon_Thunders {
   widgetClassPath=&Image
   bIsVariable=true
   parent=Overlay_DromeIcon
   widget=Image_MMObj_Icon_Thunders
   properties {brush {
     bIsDynamicallyLoaded=false
     drawAs=Image
     imageType=FullColor
     mirroring=NoMirror
     resourceName=None
     resourceObject=/NitroDrome/Gameplay/LootObject/UI/Textures/T_UI_Icon_NitroDrome_Timer_Thunders
     tiling=NoTile
     imageSize=<58 65>
     margin=<ltrb 0 0 0 0>
     outlineSettings {
      bUseBrushTransparency=false
      roundingType=HalfHeightRadius
      width=0
      color {colorUseRule=UseColor_Specified, specifiedColor=<rgba 0 0 0 0>}
      cornerRadii=<0 0 0 0>
     }
     tintColor {colorUseRule=UseColor_Specified, specifiedColor=<rgba 1 1 1 1>}
     uVRegion {bIsValid=false, max=<0 0>, min=<0 0>}
    }}
   slot {class=OverlaySlot, properties {horizontalAlignment=HAlign_Center, verticalAlignment=VAlign_Center}}
  }
 ]
}
