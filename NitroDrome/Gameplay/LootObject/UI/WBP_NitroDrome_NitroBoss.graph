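# NitroDrome boss HUD widget showing Ringmaster Scarr's icon and a banner announcing her fight or elimination.
# Play functions drive fade-in, spawn/focus and banner-hide animations, while UpdateBossAliveNotFocused hides any visible banner and restores the icon to white.
# A visibility switcher selects the banner text, and FortVisualAttachment positions it beside the icon.
# OnBossKilled plays the crossed-out icon animation, then a timer broadcasts OnBossKilledAnimFinished after animation length plus hold duration.
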WidgetBlueprint /NitroDrome/Gameplay/LootObject/UI/WBP_NitroDrome_NitroBoss
animations [
 Anim_FadeIn {
  binding_count=2
  length_seconds=0.5
  object=&self:Anim_FadeIn
  sequence {
   bindings [
    Image_Boss_Icon {
     id=#1
     possessed_class=Image
     tracks [{
       type=MovieScene2DTransformTrack
       sections [{
         type=MovieScene2DTransformSection
         end_frame=30
         start_frame=0
         channels [
          Translation.X {type=MovieSceneScriptingFloatChannel, const [0,1]}
          Translation.Y {type=MovieSceneScriptingFloatChannel, const [0,1]}
          Angle {type=MovieSceneScriptingFloatChannel, const [0,1]}
          Scale.X {type=MovieSceneScriptingFloatChannel, keys [[0,2],[15,1],[30,1]]}
          Scale.Y {type=MovieSceneScriptingFloatChannel, keys [[0,2],[15,1],[30,1]]}
          Shear.X {type=MovieSceneScriptingFloatChannel, const [0,1]}
          Shear.Y {type=MovieSceneScriptingFloatChannel, const [0,1]}
         ]
        }]
      }]
    }
    Overlay_Root {
     id=#2
     display_name=Overlay_DromeTimer
     possessed_class=Overlay
     tracks [
      {
       type=MovieSceneFloatTrack
       display_name=Render Opacity
       sections [{type=MovieSceneFloatSection, end_frame=30, start_frame=0, channels [None {type=MovieSceneScriptingFloatChannel, keys [[0,0],[21,1],[30,1]]}]}]
      }
      {
       type=MovieScene2DTransformTrack
       sections [{
         type=MovieScene2DTransformSection
         end_frame=30
         start_frame=0
         channels [
          Translation.X {type=MovieSceneScriptingFloatChannel}
          Translation.Y {type=MovieSceneScriptingFloatChannel}
         ]
        }]
      }
     ]
    }
   ]
   display_rate {numerator=60}
  }
 }
 Anim_Spawned {
  binding_count=5
  length_seconds=1.7
  object=&self:Anim_Spawned
  sequence {bindings [
    Image_Boss_Icon {
     id=#3
     possessed_class=Image
     tracks [{
       type=MovieSceneColorTrack
       display_name=Color and Opacity
       sections [{
         type=MovieSceneColorSection
         end_frame=30
         start_frame=0
         channels [
          Color.R {type=MovieSceneScriptingFloatChannel, thinned_from=4, keys [[0,0.2083],[10,1],[30,1]]}
          Color.G {type=MovieSceneScriptingFloatChannel, thinned_from=4, keys [[0,0.2083],[10,1],[30,1]]}
          Color.B {type=MovieSceneScriptingFloatChannel, thinned_from=4, keys [[0,0.2083],[10,1],[30,1]]}
          Color.A {type=MovieSceneScriptingFloatChannel, const [1,3]}
         ]
        }]
      }]
    }
    Image_Glow {
     id=#4
     possessed_class=Image
     tracks [
      {
       type=MovieSceneColorTrack
       display_name=Color and Opacity
       sections [{
         type=MovieSceneColorSection
         end_frame=0
         start_frame=-1
         channels [
          Color.R {type=MovieSceneScriptingFloatChannel, const [1,1]}
          Color.G {type=MovieSceneScriptingFloatChannel, const [1,1]}
          Color.B {type=MovieSceneScriptingFloatChannel, const [1,1]}
          Color.A {type=MovieSceneScriptingFloatChannel, const [1,1]}
         ]
        }]
      }
      {
       type=MovieSceneFloatTrack
       display_name=Render Opacity
       sections [{
         type=MovieSceneFloatSection
         end_frame=25
         start_frame=-1
         channels [None {type=MovieSceneScriptingFloatChannel, keys [[0,1],[10,1],[15.6667,0]]}]
        }]
      }
     ]
    }
    Text_Prompt_AnimOnly {
     id=#5
     possessed_class=CommonTextBlock
     tracks [
      {
       type=MovieSceneFloatTrack
       display_name=Render Opacity
       sections [{
         type=MovieSceneFloatSection
         end_frame=33
         start_frame=-1
         channels [None {type=MovieSceneScriptingFloatChannel, thinned_from=5, keys [[0,0],[13,0],[13.3333,0.5],[21,0]]}]
        }]
      }
      {
       type=MovieScene2DTransformTrack
       sections [{
         type=MovieScene2DTransformSection
         end_frame=33
         start_frame=0
         channels [
          Translation.X {type=MovieSceneScriptingFloatChannel, const [0,2]}
          Translation.Y {type=MovieSceneScriptingFloatChannel, const [0,2]}
          Angle {type=MovieSceneScriptingFloatChannel, const [0,2]}
          Scale.X {type=MovieSceneScriptingFloatChannel, thinned_from=4, keys [[0,1],[13.3333,1],[21,2]]}
          Scale.Y {type=MovieSceneScriptingFloatChannel, thinned_from=4, keys [[0,1],[13.3333,1],[21,2]]}
          Shear.X {type=MovieSceneScriptingFloatChannel, const [0,2]}
          Shear.Y {type=MovieSceneScriptingFloatChannel, const [0,2]}
         ]
        }]
      }
     ]
    }
    Text_FightBossPrompt {
     id=#6
     possessed_class=CommonTextBlock
     tracks [{
       type=MovieScene2DTransformTrack
       sections [{
         type=MovieScene2DTransformSection
         end_frame=33
         start_frame=0
         channels [
          Translation.X {type=MovieSceneScriptingFloatChannel}
          Translation.Y {type=MovieSceneScriptingFloatChannel}
          Angle {type=MovieSceneScriptingFloatChannel}
          Scale.X {type=MovieSceneScriptingFloatChannel, thinned_from=4, keys [[0,3],[14,1],[29,1]]}
          Scale.Y {type=MovieSceneScriptingFloatChannel, thinned_from=4, keys [[0,3],[14,1],[29,1]]}
          Shear.X {type=MovieSceneScriptingFloatChannel}
          Shear.Y {type=MovieSceneScriptingFloatChannel}
         ]
        }]
      }]
    }
    Overlay_Annoucement {
     id=#7
     possessed_class=Overlay
     tracks [{
       type=MovieSceneFloatTrack
       display_name=Render Opacity
       sections [{
         type=MovieSceneFloatSection
         end_frame=25
         start_frame=0
         channels [None {type=MovieSceneScriptingFloatChannel, thinned_from=5, keys [[0,0],[5,0],[8.3333,1],[25,1]]}]
        }]
      }]
    }
   ]}
 }
 Anim_BossKilled {
  binding_count=5
  length_seconds=0.7511
  object=&self:Anim_BossKilled
  sequence {
   bindings [
    Overlay_Annoucement {
     id=#8
     possessed_class=Overlay
     tracks [{
       type=MovieSceneFloatTrack
       display_name=Render Opacity
       sections [{type=MovieSceneFloatSection, channels [None {type=MovieSceneScriptingFloatChannel, keys [[0,0],[15,0],[33,1],[45,1]]}]}]
      }]
    }
    Text_BossKilledPrompt {
     id=#9
     possessed_class=CommonTextBlock
     tracks [{
       type=MovieScene2DTransformTrack
       sections [{
         type=MovieScene2DTransformSection
         channels [
          Translation.X {type=MovieSceneScriptingFloatChannel, const [0,1]}
          Translation.Y {type=MovieSceneScriptingFloatChannel, const [0,1]}
          Angle {type=MovieSceneScriptingFloatChannel, const [0,1]}
          Scale.X {type=MovieSceneScriptingFloatChannel, thinned_from=4, keys [[0,2],[23,2],[33,1]]}
          Scale.Y {type=MovieSceneScriptingFloatChannel, thinned_from=4, keys [[0,2],[23,2],[33,1]]}
          Shear.X {type=MovieSceneScriptingFloatChannel, const [0,1]}
          Shear.Y {type=MovieSceneScriptingFloatChannel, const [0,1]}
         ]
        }]
      }]
    }
    Image_X_LeftRight {
     id=#10
     possessed_class=Image
     tracks [
      {
       type=MovieScene2DTransformTrack
       sections [{
         type=MovieScene2DTransformSection
         channels [
          Translation.X {type=MovieSceneScriptingFloatChannel}
          Translation.Y {type=MovieSceneScriptingFloatChannel}
          Angle {type=MovieSceneScriptingFloatChannel}
          Scale.X {type=MovieSceneScriptingFloatChannel, keys [[9.936,0],[27.423,0],[38.709,1.2],[42.936,1],[45.066,1]]}
          Scale.Y {type=MovieSceneScriptingFloatChannel, keys [[9.936,0],[27.423,0],[38.709,1.2],[42.936,1],[45.066,1]]}
          Shear.X {type=MovieSceneScriptingFloatChannel}
          Shear.Y {type=MovieSceneScriptingFloatChannel}
         ]
        }]
      }
      {
       type=MovieSceneColorTrack
       display_name=Color and Opacity
       sections [{
         type=MovieSceneColorSection
         channels [
          Color.R {type=MovieSceneScriptingFloatChannel, const [1,1]}
          Color.G {type=MovieSceneScriptingFloatChannel, const [1,1]}
          Color.B {type=MovieSceneScriptingFloatChannel, const [1,1]}
          Color.A {type=MovieSceneScriptingFloatChannel, const [1,1]}
         ]
        }]
      }
      {
       type=MovieSceneFloatTrack
       display_name=Render Opacity
       sections [{type=MovieSceneFloatSection, channels [None {type=MovieSceneScriptingFloatChannel, const [1,2]}]}]
      }
     ]
    }
    Overlay_Main {
     id=#11
     possessed_class=Overlay
     tracks [{
       type=MovieSceneFloatTrack
       display_name=Render Opacity
       sections [{type=MovieSceneFloatSection, channels [None {type=MovieSceneScriptingFloatChannel, const [1,2]}]}]
      }]
    }
    Image_X_RightLeft {
     id=#12
     possessed_class=Image
     tracks [
      {
       type=MovieScene2DTransformTrack
       sections [{
         type=MovieScene2DTransformSection
         channels [
          Translation.X {type=MovieSceneScriptingFloatChannel}
          Translation.Y {type=MovieSceneScriptingFloatChannel}
          Angle {type=MovieSceneScriptingFloatChannel}
          Scale.X {type=MovieSceneScriptingFloatChannel, keys [[0.513,0],[18,0],[29.286,1.2],[35.643,1]]}
          Scale.Y {type=MovieSceneScriptingFloatChannel, keys [[0.513,0],[18,0],[29.286,1.2],[35.643,1]]}
          Shear.X {type=MovieSceneScriptingFloatChannel}
          Shear.Y {type=MovieSceneScriptingFloatChannel}
         ]
        }]
      }
      {
       type=MovieSceneColorTrack
       display_name=Color and Opacity
       sections [{
         type=MovieSceneColorSection
         channels [
          Color.R {type=MovieSceneScriptingFloatChannel, const [1,1]}
          Color.G {type=MovieSceneScriptingFloatChannel, const [1,1]}
          Color.B {type=MovieSceneScriptingFloatChannel, const [1,1]}
          Color.A {type=MovieSceneScriptingFloatChannel, const [1,1]}
         ]
        }]
      }
      {
       type=MovieSceneFloatTrack
       display_name=Render Opacity
       sections [{type=MovieSceneFloatSection, channels [None {type=MovieSceneScriptingFloatChannel, const [1,2]}]}]
      }
     ]
    }
   ]
   display_rate {numerator=60}
  }
 }
 Anim_HideBanner {
  binding_count=1
  length_seconds=0.5
  object=&self:Anim_HideBanner
  sequence {bindings [Overlay_Annoucement {
     id=#13
     possessed_class=Overlay
     tracks [{
       type=MovieSceneFloatTrack
       display_name=Render Opacity
       sections [{type=MovieSceneFloatSection, end_frame=10, start_frame=0, channels [None {type=MovieSceneScriptingFloatChannel, keys [[0,1],[7,0]]}]}]
      }]
    }]}
 }
]
graphs [
 EventGraph {nodes [
   Event OnBossKilled() {is=title K2Node_CustomEvent}
   > n1 = SetActiveWidget(
    self=Get VisSwticher_AnnoucementType() -> (VisSwticher_AnnoucementType:Common Visibility Switcher Object Reference)
    Widget=Get VBox_BossKilled() -> (VBox_BossKilled:Vertical Box Object Reference)
   )
   n2 = Event Call Complete Delegate() {is=title K2Node_CustomEvent}
   n3 = Get Anim_BossKilled() -> (Anim_BossKilled:Widget Animation Object Reference)
   n4 = Set Timer by Event(
    execute=@n1.then
    Delegate=@n2
    Time=float + float(A=GetEndTime(self=@n3), B=Get DurationBossKilledVisible() -> (DurationBossKilledVisible:double))
    bLooping=false
   )
   > Set Timer_BossKilledAnimFinished(Timer_BossKilledAnimFinished:Timer Handle Structure=@n4) -> (Output_Get:Timer Handle Structure)
   > PlayAnimation(InAnimation=Get Anim_BossKilled())
   Clear and Invalidate Timer by Handle(execute=@n5.then, Handle=Get Timer_BossKilledAnimFinished() -> (Timer_BossKilledAnimFinished:Timer Handle Structure))
   n5 = Call On Boss Killed Anim Finished!(execute:Exec=@n2.then, self:WBP Nitro Drome Nitro Boss Object Reference) -> (then:Exec) {is=title K2Node_CallDelegate}
  ]}
 OnBossKilledAnimFinished {nodes [OnBossKilledAnimFinished() {is=title K2Node_FunctionEntry}]}
 PlayBossSpawnedAnim {nodes [
   PlayBossSpawnedAnim() {is=title K2Node_FunctionEntry}
   > SetActiveWidget(self=Get VisSwticher_AnnoucementType(), Widget=Get VBox_BossSpawned() -> (VBox_BossSpawned:Vertical Box Object Reference))
   > PlayAnimation(InAnimation=Get Anim_Spawned() -> (Anim_Spawned:Widget Animation Object Reference))
  ]}
 PlayFadeInAnim {nodes [
   PlayFadeInAnim() {is=title K2Node_FunctionEntry}
   > PlayAnimation(InAnimation=Get Anim_FadeIn() -> (Anim_FadeIn:Widget Animation Object Reference))
  ]}
 PlayFocusBossAnim {nodes [
   PlayFocusBossAnim() {is=title K2Node_FunctionEntry}
   > PlayBossSpawnedAnim() {is=title K2Node_CallFunction}
  ]}
 PlayHideBanner {nodes [
   PlayHideBanner() {is=title K2Node_FunctionEntry}
   n1 = Get Overlay_Annoucement() -> (Overlay_Annoucement:Overlay Object Reference)
   > Branch(Condition="float > float"(A=GetRenderOpacity!(self:Widget Object Reference=@n1) -> (ReturnValue:float)))
   > PlayAnimation(InAnimation=Get Anim_HideBanner() -> (Anim_HideBanner:Widget Animation Object Reference))
  ]}
 UpdateBossAliveNotFocused {nodes [
   UpdateBossAliveNotFocused() {is=title K2Node_FunctionEntry}
   > PlayHideBanner() {is=title K2Node_CallFunction}
   > Set ColorAndOpacity!(
    execute
    ColorAndOpacity:Linear Color Structure=(R=1,G=1,B=1,A=1)
    self:Image Object Reference=Get Image_Boss_Icon() -> (Image_Boss_Icon:Image Object Reference)
   ) -> (then, Output_Get:Linear Color Structure)
  ]}
]
variables [
 DefualtBGColor {type=LinearColor, default=<rgba 0 0 0 1>}
 DurationBossKilledVisible {type=number, default=4}
 OnBossKilledAnimFinished {}
 Timer_BossKilledAnimFinished {type=TimerHandle}
]
widgets {
 info {inheritedWidgetCount=0, parentClass=&UserWidget, rootWidgetClass=&Overlay, widgetCount=21}
 widgets [
  Overlay_Root {widgetClassPath=&Overlay, bIsVariable=true, widget=Overlay_Root}
  FortVisualAttachment_0 {
   widgetClassPath=&FortVisualAttachment
   parent=Overlay_Root
   widget=FortVisualAttachment_0
   properties {contentAnchor=<-0.25 0.65>}
   slot {class=OverlaySlot, properties {horizontalAlignment=HAlign_Center, verticalAlignment=VAlign_Center}}
  }
  Overlay_Annoucement {
   widgetClassPath=&Overlay
   bIsVariable=true
   parent=FortVisualAttachment_0
   widget=Overlay_Annoucement
   properties {renderOpacity=0}
   slot {class=SizeBoxSlot, properties {horizontalAlignment=HAlign_Left, verticalAlignment=VAlign_Top}}
  }
  Image_Glow {
   widgetClassPath=&Image
   parent=Overlay_Annoucement
   widget=Image_Glow
   properties {
    renderOpacity=0.2
    brush {
     bIsDynamicallyLoaded=false
     drawAs=Image
     imageType=NoImage
     mirroring=NoMirror
     resourceName=None
     resourceObject=/Game/Athena/HUD/LTM/Wax/Materials/M_UI_TextGlow_Add
     tiling=NoTile
     imageSize=<32 32>
     margin=<ltrb 0 0 0 0>
     outlineSettings {
      bUseBrushTransparency=false
      roundingType=HalfHeightRadius
      width=0
      color {colorUseRule=UseColor_Specified, specifiedColor=<rgba 0 0 0 0>}
      cornerRadii=<0 0 0 0>
     }
     tintColor {colorUseRule=UseColor_Specified, specifiedColor=<rgba 1 1 1 1>}
     uVRegion {bIsValid=false, max=<0 0>, min=<0 0>}
    }
    colorAndOpacity=<rgba 0 0.0103 0.1301 0.5>
    renderTransform {angle=0, scale=<1 1>, shear=<0 -4>, translation=<0 0>}
   }
   slot {class=OverlaySlot, properties {horizontalAlignment=HAlign_Fill, verticalAlignment=VAlign_Fill, padding=<ltrb -75 -32 -75 -32>}}
  }
  Image_Background {
   widgetClassPath=&Image
   parent=Overlay_Annoucement
   widget=Image_Background
   properties {
    brush {
     bIsDynamicallyLoaded=false
     drawAs=Image
     imageType=FullColor
     mirroring=NoMirror
     resourceName=None
     resourceObject=/NitroDrome/Gameplay/LootObject/UI/Materials/MI_NitroDrome_TimeAffecting_Background
     tiling=NoTile
     imageSize=<32 32>
     margin=<ltrb 0 0 0 0>
     outlineSettings {
      bUseBrushTransparency=false
      roundingType=HalfHeightRadius
      width=0
      color {colorUseRule=UseColor_Specified, specifiedColor=<rgba 0 0 0 0>}
      cornerRadii=<0 0 0 0>
     }
     tintColor {colorUseRule=UseColor_Specified, specifiedColor=<rgba 1 1 1 1>}
     uVRegion {bIsValid=false, max=<0 0>, min=<0 0>}
    }
    colorAndOpacity=<rgba 0.4 0.5 0.7 1>
    renderTransform {angle=0, scale=<1 1>, shear=<0 -4>, translation=<0 0>}
   }
   slot {class=OverlaySlot, properties {horizontalAlignment=HAlign_Fill, verticalAlignment=VAlign_Fill, padding=<ltrb -75 -5 -75 -5>}}
  }
  VisSwticher_AnnoucementType {
   widgetClassPath=&CommonVisibilitySwitcher
   bIsVariable=true
   parent=Overlay_Annoucement
   widget=VisSwticher_AnnoucementType
   slot {class=OverlaySlot, properties {horizontalAlignment=HAlign_Fill, verticalAlignment=VAlign_Fill}}
  }
  VBox_BossSpawned {
   widgetClassPath=&VerticalBox
   bIsVariable=true
   parent=VisSwticher_AnnoucementType
   widget=VBox_BossSpawned
   properties {renderTransform {angle=0, scale=<1 1>, shear=<-8 -4>, translation=<0 0>}}
   slot {class=CommonVisibilitySwitcherSlot}
  }
  Overlay_0 {
   widgetClassPath=&Overlay
   parent=VBox_BossSpawned
   widget=Overlay_0
   slot {class=VerticalBoxSlot, properties {verticalAlignment=VAlign_Center}}
  }
  Text_FightBossPrompt {
   widgetClassPath=&CommonTextBlock
   parent=Overlay_0
   widget=Text_FightBossPrompt
   properties {
    justification=Center
    lineHeightPercentage=0.8
    text="Battle Ringmaster Scarr\r\nFor her Medallion!"
    textTransformPolicy=ToUpper
    font {
     bForceMonospaced=false
     bMaterialIsStencil=false
     fontMaterial=null
     fontObject=&BurbankSmall
     letterSpacing=0
     monospacedWidth=1
     size=15
     skewAmount=0
     typefaceFontName=Black
     outlineSettings {
      bApplyOutlineToDropShadows=false
      bMiteredCorners=false
      bSeparateFillAlpha=false
      outlineMaterial=null
      outlineSize=0
      outlineColor=<rgba 0 0 0 1>
     }
    }
    shadowColorAndOpacity=<rgba 0 0.0103 0.1301 0.8>
    shadowOffset=<3 4>
   }
   slot {class=OverlaySlot, properties {horizontalAlignment=HAlign_Fill, verticalAlignment=VAlign_Center}}
  }
  Text_Prompt_AnimOnly {
   widgetClassPath=&CommonTextBlock
   parent=Overlay_0
   widget=Text_Prompt_AnimOnly
   properties {
    justification=Center
    lineHeightPercentage=0.8
    text="Battle Ringmaster Scarr\r\nFor her Medallion!"
    textTransformPolicy=ToUpper
    font {
     bForceMonospaced=false
     bMaterialIsStencil=false
     fontMaterial=null
     fontObject=&BurbankSmall
     letterSpacing=0
     monospacedWidth=1
     size=15
     skewAmount=0
     typefaceFontName=Black
     outlineSettings {
      bApplyOutlineToDropShadows=false
      bMiteredCorners=false
      bSeparateFillAlpha=false
      outlineMaterial=null
      outlineSize=0
      outlineColor=<rgba 0 0 0 1>
     }
    }
    shadowColorAndOpacity=<rgba 0 0.0103 0.1301 0.8>
    shadowOffset=<3 4>
   }
   slot {class=OverlaySlot, properties {horizontalAlignment=HAlign_Fill, verticalAlignment=VAlign_Center}}
  }
  VBox_BossKilled {
   widgetClassPath=&VerticalBox
   bIsVariable=true
   parent=VisSwticher_AnnoucementType
   widget=VBox_BossKilled
   properties {renderTransform {angle=0, scale=<1 1>, shear=<-8 -4>, translation=<0 0>}}
   slot {class=CommonVisibilitySwitcherSlot}
  }
  Text_BossKilledPrompt {
   widgetClassPath=&CommonTextBlock
   parent=VBox_BossKilled
   widget=Text_BossKilledPrompt
   properties {
    justification=Center
    lineHeightPercentage=0.8
    text="Ringmaster Scarr\r\nEliminated!"
    textTransformPolicy=ToUpper
    font {
     bForceMonospaced=false
     bMaterialIsStencil=false
     fontMaterial=null
     fontObject=&BurbankSmall
     letterSpacing=0
     monospacedWidth=1
     size=15
     skewAmount=0
     typefaceFontName=Black
     outlineSettings {
      bApplyOutlineToDropShadows=false
      bMiteredCorners=false
      bSeparateFillAlpha=false
      outlineMaterial=null
      outlineSize=0
      outlineColor=<rgba 0 0 0 1>
     }
    }
    shadowColorAndOpacity=<rgba 0 0.0103 0.1301 0.8>
    shadowOffset=<3 4>
   }
   slot {class=VerticalBoxSlot, properties {verticalAlignment=VAlign_Center}}
  }
  Overlay_Main {widgetClassPath=&Overlay, bIsVariable=true, parent=Overlay_Root, widget=Overlay_Main, slot {class=OverlaySlot}}
  Image_Unlocked_Glow {
   widgetClassPath=&Image
   parent=Overlay_Main
   widget=Image_Unlocked_Glow
   properties {
    renderOpacity=0
    brush {
     bIsDynamicallyLoaded=false
     drawAs=Image
     imageType=NoImage
     mirroring=NoMirror
     resourceName=None
     resourceObject=/FortUILibrary/Styles/Border/Materials/M_Border_SoftGlow
     tiling=NoTile
     imageSize=<64 64>
     margin=<ltrb 0 0 0 0>
     outlineSettings {
      bUseBrushTransparency=false
      roundingType=HalfHeightRadius
      width=0
      color {colorUseRule=UseColor_Specified, specifiedColor=<rgba 0 0 0 0>}
      cornerRadii=<0 0 0 0>
     }
     tintColor {colorUseRule=UseColor_Specified, specifiedColor=<rgba 1 1 1 1>}
     uVRegion {bIsValid=false, max=<0 0>, min=<0 0>}
    }
    colorAndOpacity=<rgba 0.2472 1 0 1>
   }
   slot {class=OverlaySlot, properties {horizontalAlignment=HAlign_Fill, verticalAlignment=VAlign_Fill, padding=<ltrb -5 -5 -5 -5>}}
  }
  Overlay_IconBG {
   widgetClassPath=&Overlay
   bIsVariable=true
   parent=Overlay_Main
   widget=Overlay_IconBG
   slot {class=OverlaySlot, properties {horizontalAlignment=HAlign_Fill, verticalAlignment=VAlign_Fill}}
  }
  Image_Diamond_BGOpaque {
   widgetClassPath=&Image
   parent=Overlay_IconBG
   widget=Image_Diamond_BGOpaque
   properties {brush {
     bIsDynamicallyLoaded=false
     drawAs=Image
     imageType=NoImage
     mirroring=NoMirror
     resourceName=None
     resourceObject=/Game/Athena/HUD/Compass/M_UI_DiamondMarker
     tiling=NoTile
     imageSize=<46 46>
     margin=<ltrb 0 0 0 0>
     outlineSettings {
      bUseBrushTransparency=false
      roundingType=HalfHeightRadius
      width=0
      color {colorUseRule=UseColor_Specified, specifiedColor=<rgba 0 0 0 0>}
      cornerRadii=<0 0 0 0>
     }
     tintColor {colorUseRule=UseColor_Specified, specifiedColor=<rgba 0 0 0 0.502>}
     uVRegion {bIsValid=false, max=<0 0>, min=<0 0>}
    }}
   slot {class=OverlaySlot, properties {horizontalAlignment=HAlign_Fill, verticalAlignment=VAlign_Fill, padding=<ltrb -30 -30 -30 -30>}}
  }
  Image_Unlocked_IconGlow_01 {
   widgetClassPath=&Image
   parent=Overlay_Main
   widget=Image_Unlocked_IconGlow_01
   properties {
    brush {
     bIsDynamicallyLoaded=false
     drawAs=Image
     imageType=NoImage
     mirroring=NoMirror
     resourceName=None
     resourceObject=/MidMatchAssignedGameplay/Items/Environmental/Materials/MI_MMObj_Timer_Glow
     tiling=NoTile
     imageSize=<84 84>
     margin=<ltrb 0 0 0 0>
     outlineSettings {
      bUseBrushTransparency=false
      roundingType=HalfHeightRadius
      width=0
      color {colorUseRule=UseColor_Specified, specifiedColor=<rgba 0 0 0 0>}
      cornerRadii=<0 0 0 0>
     }
     tintColor {colorUseRule=UseColor_Specified, specifiedColor=<rgba 1 1 1 1>}
     uVRegion {bIsValid=false, max=<0 0>, min=<0 0>}
    }
    colorAndOpacity=<rgba 0 0.8243 1 0>
   }
   slot {class=OverlaySlot, properties {horizontalAlignment=HAlign_Center, verticalAlignment=VAlign_Center, padding=<ltrb -5 -5 -5 -5>}}
  }
  Overlay_BossIcon {
   widgetClassPath=&Overlay
   parent=Overlay_Main
   widget=Overlay_BossIcon
   slot {class=OverlaySlot, properties {horizontalAlignment=HAlign_Center, verticalAlignment=VAlign_Center}}
  }
  Image_Boss_Icon {
   widgetClassPath=&Image
   bIsVariable=true
   parent=Overlay_BossIcon
   widget=Image_Boss_Icon
   properties {brush {
     bIsDynamicallyLoaded=false
     drawAs=Image
     imageType=FullColor
     mirroring=NoMirror
     resourceName=None
     resourceObject=/NitroDrome/Gameplay/LootObject/UI/Textures/T_UI_NitroQueenIcon_128
     tiling=NoTile
     imageSize=<50.5 66>
     margin=<ltrb 0 0 0 0>
     outlineSettings {
      bUseBrushTransparency=false
      roundingType=HalfHeightRadius
      width=0
      color {colorUseRule=UseColor_Specified, specifiedColor=<rgba 0 0 0 0>}
      cornerRadii=<0 0 0 0>
     }
     tintColor {colorUseRule=UseColor_Specified, specifiedColor=<rgba 1 1 1 1>}
     uVRegion {bIsValid=false, max=<0 0>, min=<0 0>}
    }}
   slot {class=OverlaySlot, properties {horizontalAlignment=HAlign_Center, verticalAlignment=VAlign_Center}}
  }
  Image_X_RightLeft {
   widgetClassPath=&Image
   bIsVariable=true
   parent=Overlay_BossIcon
   widget=Image_X_RightLeft
   properties {
    renderOpacity=0
    brush {
     bIsDynamicallyLoaded=false
     drawAs=Image
     imageType=NoImage
     mirroring=NoMirror
     resourceName=None
     resourceObject=/Game/Athena/HUD/Discoverability/materials/Hired_NPC_Elim/MI_UI_NPCElim_RightLeftX
     tiling=NoTile
     imageSize=<90 90>
     margin=<ltrb 0 0 0 0>
     outlineSettings {
      bUseBrushTransparency=false
      roundingType=HalfHeightRadius
      width=0
      color {colorUseRule=UseColor_Specified, specifiedColor=<rgba 0 0 0 0>}
      cornerRadii=<0 0 0 0>
     }
     tintColor {colorUseRule=UseColor_Specified, specifiedColor=<rgba 1 1 1 1>}
     uVRegion {bIsValid=false, max=<0 0>, min=<0 0>}
    }
    renderTransformPivot=<1 0>
   }
   slot {class=OverlaySlot, properties {horizontalAlignment=HAlign_Center, verticalAlignment=VAlign_Center}}
  }
  Image_X_LeftRight {
   widgetClassPath=&Image
   bIsVariable=true
   parent=Overlay_BossIcon
   widget=Image_X_LeftRight
   properties {
    renderOpacity=0
    brush {
     bIsDynamicallyLoaded=false
     drawAs=Image
     imageType=NoImage
     mirroring=NoMirror
     resourceName=None
     resourceObject=/Game/Athena/HUD/Discoverability/materials/Hired_NPC_Elim/MI_UI_NPCElim_LeftRightX
     tiling=NoTile
     imageSize=<90 90>
     margin=<ltrb 0 0 0 0>
     outlineSettings {
      bUseBrushTransparency=false
      roundingType=HalfHeightRadius
      width=0
      color {colorUseRule=UseColor_Specified, specifiedColor=<rgba 0 0 0 0>}
      cornerRadii=<0 0 0 0>
     }
     tintColor {colorUseRule=UseColor_Specified, specifiedColor=<rgba 1 1 1 1>}
     uVRegion {bIsValid=false, max=<0 0>, min=<0 0>}
    }
    renderTransformPivot=<0 0>
   }
   slot {class=OverlaySlot, properties {horizontalAlignment=HAlign_Center, verticalAlignment=VAlign_Center}}
  }
 ]
}
