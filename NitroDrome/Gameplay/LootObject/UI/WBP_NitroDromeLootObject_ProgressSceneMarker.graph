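# Progress marker widget for the NitroDrome loot objective, pairing a progress child widget with a Nitro boss child widget.
# A stateful reward-status listener supplies the reward manager, whose boss-status delegate also drives UpdateState.
# UpdateState steps through objective, completion, boss and optional reset states, waiting on animations and swapping focus by scale.
# A looping timer polls progress; playspace overlap, pause, non-teammate spectating and finishing without reset control visibility.
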WidgetBlueprint /NitroDrome/Gameplay/LootObject/UI/WBP_NitroDromeLootObject_ProgressSceneMarker
animations [
 Anim_NitroDromeActive {
  binding_count=3
  length_seconds=0.3334
  object=&self:Anim_NitroDromeActive
  sequence {
   bindings [
    WBP_NitroDrome_NitroBoss {
     id=#1
     possessed_class=null
     tracks [{
       type=MovieScene2DTransformTrack
       sections [{
         type=MovieScene2DTransformSection
         end_frame=20
         start_frame=0
         channels [
          Translation.X {type=MovieSceneScriptingFloatChannel, keys [[0,45],[20,62]]}
          Translation.Y {type=MovieSceneScriptingFloatChannel, keys [[0,45],[20,62]]}
          Angle {type=MovieSceneScriptingFloatChannel, const [0,1]}
          Scale.X {type=MovieSceneScriptingFloatChannel, keys [[0,1],[20,0.6]]}
          Scale.Y {type=MovieSceneScriptingFloatChannel, keys [[0,1],[20,0.6]]}
          Shear.X {type=MovieSceneScriptingFloatChannel, const [0,1]}
          Shear.Y {type=MovieSceneScriptingFloatChannel, const [0,1]}
         ]
        }]
      }]
    }
    WBP_NitroDromeLootObject_Progress {
     id=#2
     possessed_class=WBP_NitroDromeLootObject_Progress_C
     tracks [{
       type=MovieScene2DTransformTrack
       sections [{
         type=MovieScene2DTransformSection
         end_frame=20
         start_frame=0
         channels [
          Translation.X {type=MovieSceneScriptingFloatChannel, const [0,1]}
          Translation.Y {type=MovieSceneScriptingFloatChannel, const [0,1]}
          Angle {type=MovieSceneScriptingFloatChannel, const [0,1]}
          Scale.X {type=MovieSceneScriptingFloatChannel, keys [[0,0.6],[20,1]]}
          Scale.Y {type=MovieSceneScriptingFloatChannel, keys [[0,0.6],[20,1]]}
          Shear.X {type=MovieSceneScriptingFloatChannel, const [0,1]}
          Shear.Y {type=MovieSceneScriptingFloatChannel, const [0,1]}
         ]
        }]
      }]
    }
    Image_Line {
     id=#3
     possessed_class=Image
     tracks [{
       type=MovieScene2DTransformTrack
       sections [{
         type=MovieScene2DTransformSection
         end_frame=20
         start_frame=0
         channels [
          Translation.X {type=MovieSceneScriptingFloatChannel, keys [[0,12.5],[20,22]]}
          Translation.Y {type=MovieSceneScriptingFloatChannel, keys [[0,12.5],[20,22]]}
          Angle {type=MovieSceneScriptingFloatChannel, const [-45,1]}
          Scale.X {type=MovieSceneScriptingFloatChannel, const [1,1]}
          Scale.Y {type=MovieSceneScriptingFloatChannel, const [1,1]}
          Shear.X {type=MovieSceneScriptingFloatChannel, const [0,1]}
          Shear.Y {type=MovieSceneScriptingFloatChannel, const [0,1]}
         ]
        }]
      }]
    }
   ]
   display_rate {numerator=60}
  }
 }
 Anim_BossActive {
  binding_count=3
  length_seconds=0.3334
  object=&self:Anim_BossActive
  sequence {
   bindings [
    Image_Line {
     id=#4
     possessed_class=Image
     tracks [{
       type=MovieScene2DTransformTrack
       sections [{
         type=MovieScene2DTransformSection
         end_frame=20
         start_frame=0
         channels [
          Translation.X {type=MovieSceneScriptingFloatChannel, keys [[0,22],[20,12.5]]}
          Translation.Y {type=MovieSceneScriptingFloatChannel, keys [[0,22],[20,12.5]]}
          Angle {type=MovieSceneScriptingFloatChannel, const [-45,1]}
          Scale.X {type=MovieSceneScriptingFloatChannel, const [1,1]}
          Scale.Y {type=MovieSceneScriptingFloatChannel, const [1,1]}
          Shear.X {type=MovieSceneScriptingFloatChannel, const [0,1]}
          Shear.Y {type=MovieSceneScriptingFloatChannel, const [0,1]}
         ]
        }]
      }]
    }
    WBP_NitroDrome_NitroBoss {
     id=#5
     possessed_class=null
     tracks [{
       type=MovieScene2DTransformTrack
       sections [{
         type=MovieScene2DTransformSection
         end_frame=20
         start_frame=0
         channels [
          Translation.X {type=MovieSceneScriptingFloatChannel, keys [[0,62],[20,45]]}
          Translation.Y {type=MovieSceneScriptingFloatChannel, keys [[0,62],[20,45]]}
          Angle {type=MovieSceneScriptingFloatChannel, const [0,1]}
          Scale.X {type=MovieSceneScriptingFloatChannel, keys [[0,0.6],[20,1]]}
          Scale.Y {type=MovieSceneScriptingFloatChannel, keys [[0,0.6],[20,1]]}
          Shear.X {type=MovieSceneScriptingFloatChannel, const [0,1]}
          Shear.Y {type=MovieSceneScriptingFloatChannel, const [0,1]}
         ]
        }]
      }]
    }
    WBP_NitroDromeLootObject_Progress {
     id=#6
     possessed_class=WBP_NitroDromeLootObject_Progress_C
     tracks [{
       type=MovieScene2DTransformTrack
       sections [{
         type=MovieScene2DTransformSection
         end_frame=20
         start_frame=0
         channels [
          Translation.X {type=MovieSceneScriptingFloatChannel, const [0,1]}
          Translation.Y {type=MovieSceneScriptingFloatChannel, const [0,1]}
          Angle {type=MovieSceneScriptingFloatChannel, const [0,1]}
          Scale.X {type=MovieSceneScriptingFloatChannel, keys [[0,1],[20,0.6]]}
          Scale.Y {type=MovieSceneScriptingFloatChannel, keys [[0,1],[20,0.6]]}
          Shear.X {type=MovieSceneScriptingFloatChannel, const [0,1]}
          Shear.Y {type=MovieSceneScriptingFloatChannel, const [0,1]}
         ]
        }]
      }]
    }
   ]
   display_rate {numerator=60}
  }
 }
 Anim_Hide {
  binding_count=1
  length_seconds=0.3454
  object=&self:Anim_Hide
  sequence {
   bindings [Overlay_Main {
     id=#7
     possessed_class=Overlay
     tracks [{
       type=MovieSceneFloatTrack
       display_name=Render Opacity
       sections [{type=MovieSceneFloatSection, channels [None {type=MovieSceneScriptingFloatChannel, keys [[0,1],[20.722,0]]}]}]
      }]
    }]
   display_rate {numerator=60}
  }
 }
 Anim_FadeIn_Line {
  binding_count=1
  length_seconds=0.1667
  object=&self:Anim_FadeIn_Line
  sequence {
   bindings [Image_Line {
     id=#8
     possessed_class=Image
     tracks [{
       type=MovieSceneFloatTrack
       display_name=Render Opacity
       sections [{type=MovieSceneFloatSection, channels [None {type=MovieSceneScriptingFloatChannel, keys [[0,0],[10,1]]}]}]
      }]
    }]
   display_rate {numerator=60}
  }
 }
]
graphs [
 ActivateBossKilled {nodes [
   ActivateBossKilled() {is=title K2Node_FunctionEntry}
   > n1 = Branch(Condition=Get HasBossKilledPlayed() -> (HasBossKilledPlayed:bool))
   Set HasBossKilledPlayed(execute=@n1.else, HasBossKilledPlayed:bool=true) -> (Output_Get:bool)
   > Event OnBossKilled(self:WBP Nitro Drome Nitro Boss Object Reference=Get WBP_NitroDrome_NitroBoss() -> (WBP_NitroDrome_NitroBoss:WBP Nitro Drome Nitro Boss Object Reference))
  ]}
 ActivateBossSpawnedVisuals {nodes [
   ActivateBossSpawnedVisuals() {is=title K2Node_FunctionEntry}
   > HideBossUI!(execute, self, HideWidget:bool=false) -> (then)
   > PlayHideNitroDromeBanner(self:WBP Nitro Drome Loot Object Progress Object Reference=Get WBP_NitroDromeLootObject_Progress() -> (WBP_NitroDromeLootObject_Progress:WBP Nitro Drome Loot Object Progress Object Reference))
   > PlayAnimation(InAnimation=Get Anim_BossActive() -> (Anim_BossActive:Widget Animation Object Reference))
   > PlayFocusBossAnim(self:WBP Nitro Drome Nitro Boss Object Reference=Get WBP_NitroDrome_NitroBoss())
  ]}
 ActivateObjective {nodes [
   ActivateObjective() {is=title K2Node_FunctionEntry}
   > Event OnObjectiveActivated(self:WBP Nitro Drome Loot Object Progress Object Reference=Get WBP_NitroDromeLootObject_Progress())
   > Event StartDromeProgressionTimer()
  ]}
 CacheAssignedLocalPlayerPawn {nodes [
   Inputs!() -> (In:Exec)
   > n1 = Get!(execute:Exec) -> (then:Exec, else:Exec, CachedLocalPawn:Pawn Object Reference)
   Outputs!(
    Valid:Exec=[@n5.then, @n1.then]
    CachedPawn:Pawn Object Reference=Get CachedLocalPawn() -> (CachedLocalPawn:Pawn Object Reference)
    Invalid:Exec=[@n4.CastFailed, @n3.Is Not Valid]
   )
   n2 = GetOwningPlayer()
   n3 = Is Valid(exec=@n1.else, InputObject=@n2)
   > n4 = Cast To Pawn!(execute, Object=GetViewTarget!(self:Controller Object Reference=@n2) -> (ReturnValue:Actor Object Reference)) -> (then, CastFailed, AsPawn:Pawn Object Reference) {is=title K2Node_DynamicCast}
   > n5 = Set CachedLocalPawn(CachedLocalPawn:Pawn Object Reference=@n4) -> (Output_Get:Pawn Object Reference)
  ]}
 Check Widget Visibility {nodes [
   Check Widget Visibility() {is=title K2Node_FunctionEntry}
   > n1 = Branch(Condition=ShouldBeVisible!(self) -> (ShouldBeVisible:bool))
   > ShowWidget()
   HideWidget(execute=@n1.else)
  ]}
 CheckOverlappingCurrentPlayspace {nodes [
   CheckOverlappingCurrentPlayspace() {is=title K2Node_FunctionEntry}
   > n1 = Is Valid(InputObject=Get CachedPlayspaceVolume() -> (CachedPlayspaceVolume:Gameplay Volume Object Reference))
   HideWidget(execute=@n1.Is Not Valid)
   n2 = Cache Assigned Local Player Pawn!(In:Exec=@n1.Is Valid) -> (Valid:Exec, CachedPawn:Pawn Object Reference, Invalid:Exec) {is=title K2Node_MacroInstance}
   > n3 = Branch(Condition=IsOverlappingActor!(self:Actor Object Reference=Get CachedPlayspaceVolume(), Other:Actor Object Reference=@n2) -> (ReturnValue:bool))
   > n4 = Set IsPlayerWIthinPlayspace(IsPlayerWIthinPlayspace:bool=true) -> (Output_Get:bool)
   Check Widget Visibility(execute=[@n4.then, @n5.then]) {is=title K2Node_CallFunction}
   n5 = Set IsPlayerWIthinPlayspace(execute=[@n3.else, @n2.Invalid], IsPlayerWIthinPlayspace=false)
  ]}
 EventGraph {nodes [
   Event StartDromeProgressionTimer() {is=title K2Node_CustomEvent}
   > n1 = Set Timer by Event(Delegate=Create Event(), Time=Get TimerPollingUpdateInterval() -> (TimerPollingUpdateInterval:double), bLooping=true)
   > Set Timer_UpdateDromeProgressTime(Timer_UpdateDromeProgressTime:Timer Handle Structure=@n1) -> (Output_Get:Timer Handle Structure)
   Event PollingUpdateDromeProgressVisualTimer() {is=title K2Node_CustomEvent}
   > Event UpdateProgress!(
    execute
    self:WBP Nitro Drome Loot Object Progress Object Reference=Get WBP_NitroDromeLootObject_Progress()
    ProgressPercent:double=GetProgressPercentage!(self:Fort Nitro Drome Reward Manager Object Reference=Get RewardManager() -> (RewardManager:BP Nitro Drome Reward Manager Object Reference)) -> (ReturnValue:float)
   ) -> (then)
   ListenToRewardStatusEvents() {is=title K2Node_CustomEvent}
   > n2 = StartListeningToStatefulEvent!(
    execute:Exec
    Target:Gameplay Event Router Component Object Reference=GetGlobalEventRouter()
    EventType:Script Struct Object Reference=/Script/NitroDromeRuntime.OnDromeRewardStatusChanged
    Context:Object Reference
    EventBubblingRule:EEventBubblingRule Enum
    None:%=GetOwningPlayer()
   ) -> (
    then:Exec
    AsyncTaskProxy:Async Action Start Listening To Stateful Event Object Reference
    OnEventReceived:Exec
    OnSavedEventStateExists:Exec
    OnEventStateCleared:Exec
    Payload_Status:ENitroDromeRewardStatus Enum
    Payload_OldStatus:ENitroDromeRewardStatus Enum
    Payload_RewardManager:Fort Nitro Drome Reward Manager Object Reference
   ) {is=title K2Node_StartListeningToStatefulEventAsyncAction}
   > Set AsyncTaskRef_RewardStatus(AsyncTaskRef_RewardStatus:Async Action Start Listening To Stateful Event Object Reference=@n2) -> (Output_Get:Async Action Start Listening To Stateful Event Object Reference)
   Set ObjectiveState(execute=[@n2.OnEventReceived, @n2.OnSavedEventStateExists], ObjectiveState:ENitroDromeRewardStatus Enum=@n2.Payload_Status) -> (Output_Get:ENitroDromeRewardStatus Enum=Inactive)
   > SetRewardManager!(
    execute
    self
    NewRewardManager:BP Nitro Drome Reward Manager Object Reference=Cast To BP_NitroDromeRewardManager!(Object=@n2.Payload_RewardManager) -> (AsBP Nitro Drome Reward Manager:BP Nitro Drome Reward Manager Object Reference, bSuccess:bool=~) {is=title K2Node_DynamicCast}
   ) -> (then)
   > Check Widget Visibility()
   > UpdateState()
   Event On Boss Killed() {is=title K2Node_CustomEvent}
   > Set IsStateAnimating(IsStateAnimating:bool=false) -> (Output_Get:bool)
   > HideBossUI(HideWidget=true)
   > UpdateState()
   ListenToDistanceEvents() {is=title K2Node_CustomEvent}
   > FindAndBindToPlayspace()
   > CheckOverlappingCurrentPlayspace() {is=title K2Node_CallFunction}
   OnObjectiveCompleted() {is=title K2Node_CustomEvent}
   > Set IsStateAnimating(IsStateAnimating=true)
   > Event PlayCompleteAnimation(self:WBP Nitro Drome Loot Object Progress Object Reference=Get WBP_NitroDromeLootObject_Progress())
   > Clear and Invalidate Timer by Handle(Handle=Get Timer_UpdateDromeProgressTime() -> (Timer_UpdateDromeProgressTime:Timer Handle Structure))
   n3 = OnBossStatusChanged!() -> (OutputDelegate, then, NewBossStatus:FortNitroQueenStatus Enum=NewEnumerator0) {is=title K2Node_CustomEvent}
   > Set BossState(BossState:FortNitroQueenStatus Enum=@n3.NewBossStatus) -> (Output_Get:FortNitroQueenStatus Enum=NewEnumerator0)
   > UpdateState()
   n4 = Event On Complete Anims Finished() {is=title K2Node_CustomEvent}
   n5 = Event Clear Drome Complete Banner() {is=title K2Node_CustomEvent}
   n6 = Set Timer by Event(execute=@n4.then, Delegate=@n5, Time=Get ObjectiveCompletionBannerDuration() -> (ObjectiveCompletionBannerDuration:float), bLooping=false)
   > Set Timer_ActiveStateDuration(Timer_ActiveStateDuration:Timer Handle Structure=@n6) -> (Output_Get:Timer Handle Structure)
   n7 = Event OnViewTargetChanged!() -> (
    OutputDelegate
    then
    PlayerController:Fort Player Controller Object Reference
    OldViewTarget:Actor Object Reference
    NewViewTarget:Actor Object Reference
   ) {is=title K2Node_CustomEvent}
   > n8 = Cast To FortPlayerControllerAthena(Object=@n7.PlayerController)
   > n9 = Cast To FortPlayerPawn(Object=@n7.NewViewTarget)
   n10 = GetActorTeam!(Actor:Actor Object Reference=@n9) -> (ReturnValue:Byte=0)
   n11 = GetActorTeam(Actor=@n8) -> (ReturnValue=0)
   > Set IsSpectatingNonTeammate(IsSpectatingNonTeammate:bool=AND Boolean(
     A=GetIsSpectating!(self:Fort Player Controller Athena Object Reference=@n8) -> (ReturnValue:bool)
     B="Not Equal (Byte)"(A:Byte=@n11, B:Byte=@n10) -> (ReturnValue:bool=~) {is=title K2Node_PromotableOperator}
    )) -> (Output_Get:bool)
   > Set CachedLocalPawn()
   > CheckOverlappingCurrentPlayspace()
   Event Destruct()
   > Parent: Destruct()
   > Clear and Invalidate Timer by Handle(Handle=Get Timer_UpdateDromeProgressTime())
   > StopListeningToStatefulEvent(self:Async Action Start Listening To Stateful Event Object Reference=Get AsyncTaskRef_RewardStatus() -> (AsyncTaskRef_RewardStatus:Async Action Start Listening To Stateful Event Object Reference))
   > UnbindCachedPlayspaceVolume()
   > n12 = Cast To FortPlayerController(Object=GetOwningPlayer())
   > Unbind Event from On View Target Changed(self:Fort Player Controller Object Reference=@n12, Delegate=Create Event()) {is=title K2Node_RemoveDelegate}
   Event OnInitialized()
   > SetVisibility(InVisibility=Collapsed)
   > Event UpdateProgress(self=Get WBP_NitroDromeLootObject_Progress(), ProgressPercent=0)
   > BindToAnimationFinished(Animation=Get Anim_Hide() -> (Anim_Hide:Widget Animation Object Reference), Delegate=Create Event())
   > Bind Event to Complete Anim Finished(self:WBP Nitro Drome Loot Object Progress Object Reference=Get WBP_NitroDromeLootObject_Progress(), Delegate=Create Event()) {is=title K2Node_AddDelegate}
   > Bind Event to On Boss Killed Anim Finished(self:WBP Nitro Drome Nitro Boss Object Reference=Get WBP_NitroDrome_NitroBoss(), Delegate=Create Event()) {is=title K2Node_AddDelegate}
   Event Construct()
   > Parent: Construct()
   > ListenToRewardStatusEvents() {is=title K2Node_CallFunction}
   > n13 = Sequence()
   > n14 = Cast To FortPlayerController(Object=GetOwningPlayer())
   > Bind Event to On View Target Changed(self:Fort Player Controller Object Reference=@n14, Delegate=Create Event()) {is=title K2Node_AddDelegate}
   Is Valid(exec=@n13.then_1, InputObject=Get RewardManager())
   > ListenToDistanceEvents() {is=title K2Node_CallFunction}
   UpdateState(execute=@n15.then)
   PlayHideNitroDromeBanner(execute=@n5.then, self=Get WBP_NitroDromeLootObject_Progress())
   > n15 = Set IsStateAnimating(IsStateAnimating=false)
  ]}
 FindAndBindToPlayspace {nodes [
   FindAndBindToPlayspace() {is=title K2Node_FunctionEntry}
   n1 = Get RewardManager()
   > Is Valid(InputObject=@n1)
   > n2 = GetPlayspaceForLocation!(execute, WorldContextObject:Object Reference=@n1, InLocation:"Vector (by ref)"=Get Actor Location(self=@n1)) -> (then, ReturnValue:Playspace Object Reference)
   > n3 = Cast To BP_NitroDromePlayspace!(execute, Object=@n2) -> (then, CastFailed, AsBP Nitro Drome Playspace:BP Nitro Drome Playspace Object Reference) {is=title K2Node_DynamicCast}
   > n4 = Set NewPlayspaceVolume(NewPlayspaceVolume:Gameplay Volume Object Reference=Get BoundGameplayVolume!(self:Playspace Object Reference=@n3) -> (BoundGameplayVolume:Gameplay Volume Object Reference)) -> (Output_Get:Gameplay Volume Object Reference)
   Branch(
    execute=[@n4.then, @n3.CastFailed]
    Condition="Not Equal (Object)"(
     A:Object Reference=Get NewPlayspaceVolume() -> (NewPlayspaceVolume:Gameplay Volume Object Reference)
     B:Object Reference=Get CachedPlayspaceVolume()
    ) -> (ReturnValue:bool=~) {is=title K2Node_PromotableOperator}
   )
   > UnbindCachedPlayspaceVolume()
   > n5 = Set CachedPlayspaceVolume(CachedPlayspaceVolume:Gameplay Volume Object Reference=Get NewPlayspaceVolume()) -> (Output_Get:Gameplay Volume Object Reference)
   > Is Valid(InputObject=@n5)
   > Bind Event to On Actor Begin Overlap(self:Actor Object Reference=@n5, Delegate=Create Event()) {is=title K2Node_AddDelegate}
   > Bind Event to On Actor End Overlap(self:Actor Object Reference=@n5, Delegate=Create Event()) {is=title K2Node_AddDelegate}
  ]}
 HideBossUI {nodes [
   n1 = HideBossUI!() -> (then, HideWidget:bool=~) {is=title K2Node_FunctionEntry}
   > SetVisibility(
    self=Get Overlay_BossUI() -> (Overlay_BossUI:Overlay Object Reference)
    InVisibility=Select(Option 0=HitTestInvisible, Option 1=Collapsed, Index=@n1)
   )
  ]}
 HideWidget {nodes [
   HideWidget() {is=title K2Node_FunctionEntry}
   > Branch(Condition=Get IsWidgetVisible() -> (IsWidgetVisible:bool))
   > Set IsWidgetVisible(IsWidgetVisible:bool=false) -> (Output_Get:bool)
   > Set IsStateAnimating(IsStateAnimating=false)
   > Clear and Invalidate Timer by Handle(Handle=Get Timer_UpdateDromeProgressTime())
   > Clear and Invalidate Timer by Handle(Handle=Get Timer_ActiveStateDuration() -> (Timer_ActiveStateDuration:Timer Handle Structure))
   > OnWidgetHidden(self:WBP Nitro Drome Loot Object Progress Object Reference=Get WBP_NitroDromeLootObject_Progress())
   > PlayAnimation(InAnimation=Get Anim_Hide())
  ]}
 IsResetEnabled {nodes [
   IsResetEnabled() {is=title K2Node_FunctionEntry}
   > Return Node!(
    execute
    IsResetEnabled:bool=GetValueAsBool!(Input:"Scalable Float Structure (by ref)"=Get SF_ResetEnabled() -> (SF_ResetEnabled:Scalable Float Structure), Level:float=0) -> (ReturnValue:bool)
   )
  ]}
 OnHideAnimationFinished {nodes [
   OnHideAnimationFinished() {is=title K2Node_FunctionEntry}
   > n1 = Branch(Condition=Get IsWidgetVisible())
   SetVisibility(execute=@n1.else, InVisibility=Collapsed)
  ]}
 OnOverlapWithPlayspaceBegin {nodes [
   n1 = OnOverlapWithPlayspaceBegin!() -> (then, OverlappedActor:Actor Object Reference, OtherActor:Actor Object Reference) {is=title K2Node_FunctionEntry}
   > n2 = Cache Assigned Local Player Pawn()
   > Branch(Condition="Equal (Object)"(A=@n2, B=@n1.OtherActor))
   > Set IsPlayerWIthinPlayspace(IsPlayerWIthinPlayspace=true)
   > Check Widget Visibility()
  ]}
 OnOverlapWithPlayspaceEnd {nodes [
   n1 = OnOverlapWithPlayspaceEnd!() -> (then, OverlappedActor:Actor Object Reference, OtherActor:Actor Object Reference) {is=title K2Node_FunctionEntry}
   > n2 = Cache Assigned Local Player Pawn()
   > Branch(Condition="Equal (Object)"(A=@n2, B=@n1.OtherActor))
   > Set IsPlayerWIthinPlayspace(IsPlayerWIthinPlayspace=false)
   > Check Widget Visibility()
  ]}
 OnTimerAffectingActionOccurred {nodes [
   n1 = OnTimerAffectingActionOccurred!() -> (then, ReducedTimeDelta:double=~, AffectingTimePlayer:Actor Object Reference, AffectingEventTags:Gameplay Tag Container Structure) {is=title K2Node_FunctionEntry}
   > OnTimeAffectedEventTriggered!(
    execute
    self:WBP Nitro Drome Loot Object Progress Object Reference=Get WBP_NitroDromeLootObject_Progress()
    InTimeDelta:double=@n1
    InPlayer:Actor Object Reference=@n1.AffectingTimePlayer
    InEventTags:Gameplay Tag Container Structure=@n1.AffectingEventTags
   ) -> (then)
  ]}
 SetRewardManager {nodes [
   n1 = SetRewardManager!() -> (then, NewRewardManager:BP Nitro Drome Reward Manager Object Reference) {is=title K2Node_FunctionEntry}
   > n2 = Branch(Condition="Not Equal (Object)"(A=@n1, B=Get RewardManager()) -> (ReturnValue=~))
   n3 = Get RewardManager()
   > n4 = Is Valid(InputObject=@n3)
   > Unbind Event from On Timer Affecting Action Occurred(self:BP Nitro Drome Reward Manager Object Reference=@n3, Delegate=Create Event()) {is=title K2Node_RemoveDelegate}
   > n5 = Unbind Event from Client on Boss State Changed(self:BP Nitro Drome Reward Manager Object Reference=@n3, Delegate=Create Event()) {is=title K2Node_RemoveDelegate}
   n6 = Set RewardManager(
    execute=[@n5.then, @n4.Is Not Valid]
    RewardManager:BP Nitro Drome Reward Manager Object Reference=Get NewRewardManager() -> (NewRewardManager:BP Nitro Drome Reward Manager Object Reference)
   ) -> (Output_Get:BP Nitro Drome Reward Manager Object Reference)
   > Is Valid(InputObject=@n6)
   > Bind Event to On Timer Affecting Action Occurred(self:BP Nitro Drome Reward Manager Object Reference=@n6, Delegate=Create Event()) {is=title K2Node_AddDelegate}
   > Bind Event to Client on Boss State Changed(self:BP Nitro Drome Reward Manager Object Reference=@n6, Delegate=Create Event()) {is=title K2Node_AddDelegate}
   > OnBossStatusChanged!(
    execute
    self
    NewBossStatus:FortNitroQueenStatus Enum=Get BossStatus!(self:BP Nitro Drome Reward Manager Object Reference=@n6) -> (BossStatus:FortNitroQueenStatus Enum=NewEnumerator0)
   ) -> (then) {is=title K2Node_CallFunction}
   > ListenToDistanceEvents()
   n7 = Is Valid(exec=@n2.else, InputObject=Get CachedPlayspaceVolume())
   ListenToDistanceEvents(execute=@n7.Is Not Valid)
  ]}
 ShouldBeVisible {nodes [
   ShouldBeVisible() {is=title K2Node_FunctionEntry}
   n1 = Get IsFinishedWithoutReset() -> (IsFinishedWithoutReset:bool)
   n2 = Get IsSpectatingNonTeammate() -> (IsSpectatingNonTeammate:bool)
   n3 = Get ObjectiveState() -> (ObjectiveState:ENitroDromeRewardStatus Enum=Inactive)
   > Return Node!(
    execute
    ShouldBeVisible:bool=AND Boolean!(
     A="Not Equal (Enum)"!(A:ENitroDromeRewardStatus Enum=@n3, B:ENitroDromeRewardStatus Enum=Paused) -> (ReturnValue:bool)
     B=Get IsPlayerWIthinPlayspace() -> (IsPlayerWIthinPlayspace:bool)
     C:bool=NOT Boolean(A=@n2)
     D:bool=NOT Boolean(A=@n1)
    ) -> (ReturnValue)
   )
  ]}
 ShowWidget {nodes [
   ShowWidget() {is=title K2Node_FunctionEntry}
   > n1 = Branch(Condition=Get IsWidgetVisible())
   Set IsWidgetVisible(execute=@n1.else, IsWidgetVisible=true)
   > n2 = Sequence()
   n3 = Get Anim_Hide()
   > Branch(Condition=IsAnimationPlaying(InAnimation=@n3))
   > StopAnimation(InAnimation=@n3)
   SetVisibility(execute=@n2.then_1, InVisibility=HitTestInvisible)
   > SetRenderOpacity(self=Get Overlay_Main() -> (Overlay_Main:Overlay Object Reference), InOpacity=1)
   > Set IsStateAnimating(IsStateAnimating=false)
   > n4 = Sequence()
   > n5 = Branch(Condition="Equal (Enum)"(
     A:FortNitroQueenStatus Enum=Get BossState() -> (BossState:FortNitroQueenStatus Enum=NewEnumerator0)
     B:FortNitroQueenStatus Enum=NewEnumerator2
    ))
   > HideBossUI(HideWidget=true) {is=title K2Node_CallFunction}
   PlayAnimationForward(execute=@n5.else, InAnimation=Get Anim_FadeIn_Line() -> (Anim_FadeIn_Line:Widget Animation Object Reference))
   PlayFadeInAnim(execute=@n4.then_1, self:WBP Nitro Drome Loot Object Progress Object Reference=Get WBP_NitroDromeLootObject_Progress())
   > PlayFadeInAnim(self:WBP Nitro Drome Nitro Boss Object Reference=Get WBP_NitroDrome_NitroBoss())
   > n6 = Branch(Condition="Equal (Enum)"(A:ENitroDromeRewardStatus Enum=Get ObjectiveState() -> (ObjectiveState=Inactive), B:ENitroDromeRewardStatus Enum=Active))
   > Set UIState(UIState:ENUM_NitroDrome_UIState Enum=NewEnumerator0) -> (Output_Get:ENUM_NitroDrome_UIState Enum=NewEnumerator2)
   > ActivateObjective() {is=title K2Node_CallFunction}
   n7 = Get BossState() -> (BossState=NewEnumerator0)
   n8 = Get ObjectiveState() -> (ObjectiveState=Inactive)
   n9 = Branch(
    execute=@n6.else
    Condition=AND Boolean!(A="Equal (Enum)"(A=@n8, B=Finished), B="Equal (Enum)"(A:FortNitroQueenStatus Enum=@n7, B:FortNitroQueenStatus Enum=NewEnumerator0)) -> (ReturnValue)
   )
   > Set UIState(UIState=NewEnumerator3) -> (Output_Get=NewEnumerator2)
   > OnObjectiveCompleted() {is=title K2Node_CallFunction}
   n10 = Branch(execute=@n9.else, Condition="Equal (Enum)"(A=Get BossState() -> (BossState=NewEnumerator0), B=NewEnumerator1))
   > Set UIState(UIState=NewEnumerator1) -> (Output_Get=NewEnumerator2)
   > ActivateBossSpawnedVisuals() {is=title K2Node_CallFunction}
   > SetUnlockedState!(execute, self:WBP Nitro Drome Loot Object Progress Object Reference=Get WBP_NitroDromeLootObject_Progress(), IsUnlocked:bool=true) -> (then)
   n11 = Get UIState() -> (UIState:ENUM_NitroDrome_UIState Enum=NewEnumerator2)
   n12 = Branch(
    execute=@n10.else
    Condition=AND Boolean(
     A="Equal (Enum)"(A:ENUM_NitroDrome_UIState Enum=@n11, B:ENUM_NitroDrome_UIState Enum=NewEnumerator2)
     B=IsResetEnabled!(self) -> (IsResetEnabled:bool) {is=title K2Node_CallFunction}
    )
   )
   > Event ShowRestartingBanner(self:WBP Nitro Drome Loot Object Progress Object Reference=Get WBP_NitroDromeLootObject_Progress())
   UpdateState(execute=@n12.else)
  ]}
 UnbindCachedPlayspaceVolume {nodes [
   UnbindCachedPlayspaceVolume() {is=title K2Node_FunctionEntry}
   n1 = Get CachedPlayspaceVolume()
   > Is Valid(InputObject=@n1)
   > Unbind Event from On Actor Begin Overlap(self:Actor Object Reference=@n1, Delegate=Create Event()) {is=title K2Node_RemoveDelegate}
   > Unbind Event from On Actor End Overlap(self:Actor Object Reference=@n1, Delegate=Create Event()) {is=title K2Node_RemoveDelegate}
   > Set CachedPlayspaceVolume()
  ]}
 UpdateState {nodes [
   UpdateState() {is=title K2Node_FunctionEntry}
   n1 = Get IsStateAnimating() -> (IsStateAnimating:bool)
   > Branch(Condition=AND Boolean(A=NOT Boolean(A=@n1), B=Get IsWidgetVisible()))
   > n2 = Branch(Condition="Equal (Enum)"(A:ENitroDromeRewardStatus Enum=Get ObjectiveState() -> (ObjectiveState=Inactive), B:ENitroDromeRewardStatus Enum=Active))
   > n3 = Sequence()
   > Branch(Condition="Not Equal (Enum)"(A:ENUM_NitroDrome_UIState Enum=Get UIState() -> (UIState=NewEnumerator2), B:ENUM_NitroDrome_UIState Enum=NewEnumerator0))
   > n4 = Switch on ENUM_NitroDrome_UIState!(execute, Selection:ENUM_NitroDrome_UIState Enum=Get UIState() -> (UIState=NewEnumerator2)) -> (NewEnumerator2:Exec, NewEnumerator0:Exec, NewEnumerator3:Exec, NewEnumerator1:Exec, NewEnumerator4:Exec) {is=title K2Node_SwitchEnum}
   n5 = UpdateBossAliveNotFocused(execute=@n4.NewEnumerator1, self:WBP Nitro Drome Nitro Boss Object Reference=Get WBP_NitroDrome_NitroBoss())
   n6 = PlayAnimation(execute=[@n5.then, @n4.NewEnumerator4], InAnimation=Get Anim_NitroDromeActive() -> (Anim_NitroDromeActive:Widget Animation Object Reference))
   Set UIState(execute=[@n6.then, @n4.NewEnumerator0, @n4.NewEnumerator2, @n4.NewEnumerator3], UIState=NewEnumerator0) -> (Output_Get=NewEnumerator2)
   > ActivateObjective()
   Branch(
    execute=@n3.then_1
    Condition="Equal (Enum)"(A:FortNitroQueenStatus Enum=Get BossState() -> (BossState=NewEnumerator0), B:FortNitroQueenStatus Enum=NewEnumerator2)
   )
   > ActivateBossKilled() {is=title K2Node_CallFunction}
   n7 = Get ObjectiveState() -> (ObjectiveState=Inactive)
   n8 = Get UIState() -> (UIState=NewEnumerator2)
   n9 = Branch(
    execute=@n2.else
    Condition=AND Boolean(
     A="Equal (Enum)"(A:ENitroDromeRewardStatus Enum=@n7, B:ENitroDromeRewardStatus Enum=Finished)
     B="Equal (Enum)"(A:ENUM_NitroDrome_UIState Enum=@n8, B:ENUM_NitroDrome_UIState Enum=NewEnumerator0)
    )
   )
   > Set UIState(UIState=NewEnumerator3) -> (Output_Get=NewEnumerator2)
   > OnObjectiveCompleted()
   n10 = Get BossState() -> (BossState=NewEnumerator0)
   n11 = Get UIState() -> (UIState=NewEnumerator2)
   n12 = Branch(
    execute=@n9.else
    Condition=AND Boolean(
     A="Equal (Enum)"(A:FortNitroQueenStatus Enum=@n10, B:FortNitroQueenStatus Enum=NewEnumerator1)
     B="Equal (Enum)"(A:ENUM_NitroDrome_UIState Enum=@n11, B:ENUM_NitroDrome_UIState Enum=NewEnumerator3)
    )
   )
   > Set UIState(UIState=NewEnumerator1) -> (Output_Get=NewEnumerator2)
   > ActivateBossSpawnedVisuals()
   n13 = Get UIState() -> (UIState=NewEnumerator2)
   n14 = Get BossState() -> (BossState=NewEnumerator0)
   n15 = Branch(
    execute=@n12.else
    Condition=AND Boolean(
     A="Equal (Enum)"(A:FortNitroQueenStatus Enum=@n14, B:FortNitroQueenStatus Enum=NewEnumerator2)
     B="Equal (Enum)"(A:ENUM_NitroDrome_UIState Enum=@n13, B:ENUM_NitroDrome_UIState Enum=NewEnumerator1)
    )
   )
   > Set UIState(UIState=NewEnumerator4) -> (Output_Get=NewEnumerator2)
   > Set IsStateAnimating(IsStateAnimating=true)
   > ActivateBossKilled()
   n16 = Get BossState() -> (BossState=NewEnumerator0)
   n17 = Get ObjectiveState() -> (ObjectiveState=Inactive)
   Branch(
    execute=@n15.else
    Condition=AND Boolean(
     A="Equal (Enum)"(A:FortNitroQueenStatus Enum=@n16, B:FortNitroQueenStatus Enum=NewEnumerator2)
     B="Not Equal (Enum)"(A:ENitroDromeRewardStatus Enum=@n17, B:ENitroDromeRewardStatus Enum=Active)
    )
   )
   > n18 = Branch(Condition=IsResetEnabled())
   > n19 = Branch(Condition="Equal (Enum)"(A:ENUM_NitroDrome_UIState Enum=Get UIState() -> (UIState=NewEnumerator2), B:ENUM_NitroDrome_UIState Enum=NewEnumerator4))
   > n20 = PlayAnimation(InAnimation=Get Anim_NitroDromeActive())
   Set UIState(execute=[@n20.then, @n21.else], UIState=NewEnumerator2) -> (Output_Get=NewEnumerator2)
   > Event ShowRestartingBanner(self=Get WBP_NitroDromeLootObject_Progress())
   n21 = Branch(execute=@n19.else, Condition="Equal (Enum)"(A=Get UIState() -> (UIState=NewEnumerator2), B=NewEnumerator2))
   Set IsFinishedWithoutReset(execute=@n18.else, IsFinishedWithoutReset:bool=true) -> (Output_Get:bool)
   > UnbindCachedPlayspaceVolume() {is=title K2Node_CallFunction}
   > Check Widget Visibility()
  ]}
]
variables [
 AsyncTaskRef_RewardStatus {type=AsyncAction_StartListeningToStatefulEvent, default=null}
 BossState {type=FortNitroQueenStatus, default=Inactive}
 CachedLocalPawn {type=Pawn, default=null}
 CachedPlayspaceVolume {type=GameplayVolume, default=null}
 HasBossKilledPlayed {type=boolean}
 IsFinishedWithoutReset {type=boolean}
 IsPlayerWIthinPlayspace {type=boolean}
 IsSpectatingNonTeammate {type=boolean}
 IsStateAnimating {type=boolean}
 IsWidgetVisible {type=boolean}
 ObjectiveCompletionBannerDuration {type=number, default=8}
 ObjectiveState {type=ENitroDromeRewardStatus, default=Inactive}
 RewardManager {type=BP_NitroDromeRewardManager_C, default=null}
 SF_ResetEnabled {
  type=ScalableFloat
  default {registryType=None, value=1, curve {curveTable=/NitroDrome/DataTables/NitroDromeGameData, rowName=NitroDrome.Objective.Reset.Enabled}}
 }
 Timer_ActiveStateDuration {type=TimerHandle}
 Timer_UpdateDromeProgressTime {type=TimerHandle}
 TimerPollingUpdateInterval {type=number, default=0.2}
 UIState {type=ENUM_NitroDrome_UIState, default=WaitingForReset}
]
widgets {
 info {inheritedWidgetCount=0, parentClass=&UserWidget, rootWidgetClass=&Overlay, widgetCount=6}
 widgets [
  Overlay_Root {widgetClassPath=&Overlay, widget=Overlay_Root}
  Overlay_Main {
   widgetClassPath=&Overlay
   bIsVariable=true
   parent=Overlay_Root
   widget=Overlay_Main
   slot {class=OverlaySlot, properties {horizontalAlignment=HAlign_Center, verticalAlignment=VAlign_Center, padding=<ltrb 0 20 0 30>}}
  }
  Overlay_BossUI {
   widgetClassPath=&Overlay
   bIsVariable=true
   parent=Overlay_Main
   widget=Overlay_BossUI
   slot {class=OverlaySlot, properties {horizontalAlignment=HAlign_Right, verticalAlignment=VAlign_Bottom}}
  }
  WBP_NitroDrome_NitroBoss {
   widgetClassPath=&/WBP_NitroDrome_NitroBoss
   bIsVariable=true
   parent=Overlay_BossUI
   widget=WBP_NitroDrome_NitroBoss
   properties {renderTransform {angle=0, scale=<0.6 0.6>, shear=<0 0>, translation=<62 62>}, renderTransformPivot=<0 0>}
   slot {class=OverlaySlot}
  }
  Image_Line {
   widgetClassPath=&Image
   bIsVariable=true
   parent=Overlay_BossUI
   widget=Image_Line
   properties {
    renderOpacity=0
    brush {
     bIsDynamicallyLoaded=false
     drawAs=Image
     imageType=FullColor
     mirroring=NoMirror
     resourceName=None
     resourceObject=/NitroDrome/Gameplay/LootObject/UI/Textures/T_UI_NitroDrome_Line
     tiling=NoTile
     imageSize=<17 17>
     margin=<ltrb 0 0 0 0>
     outlineSettings {
      bUseBrushTransparency=false
      roundingType=HalfHeightRadius
      width=0
      color {colorUseRule=UseColor_Specified, specifiedColor=<rgba 0 0 0 0>}
      cornerRadii=<0 0 0 0>
     }
     tintColor {colorUseRule=UseColor_Specified, specifiedColor=<rgba 1 1 1 1>}
     uVRegion {bIsValid=false, max=<0 0>, min=<0 0>}
    }
    renderTransform {angle=-45, scale=<1 1>, shear=<0 0>, translation=<22 22>}
   }
   slot {class=OverlaySlot, properties {horizontalAlignment=HAlign_Center, verticalAlignment=VAlign_Center}}
  }
  WBP_NitroDromeLootObject_Progress {
   widgetClassPath=&/WBP_NitroDromeLootObject_Progress
   bIsVariable=true
   parent=Overlay_Main
   widget=WBP_NitroDromeLootObject_Progress
   slot {class=OverlaySlot, properties {horizontalAlignment=HAlign_Center, verticalAlignment=VAlign_Center}}
  }
 ]
}
